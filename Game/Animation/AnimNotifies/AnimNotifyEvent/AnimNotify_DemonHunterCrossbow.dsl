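; Anim notify that triggers the fire-tip flare effect on the Demon Hunter crossbow during an animation.
; Received_Notify casts the mesh component's owner to B_DemonHunter_Crossbow_Athena and activates its P_DemonHunter_Crossbow_FireTipFlare component without a reset.
; A failed cast does nothing, and the notify always returns false.
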(fn Received_Notify
  (params (MeshComp :type Skeletal%20Mesh%20Component%20Object%20Reference) (Animation :type Anim%20Sequence%20Base%20Object%20Reference) (EventReference :type Anim%20Notify%20Event%20Reference%20Structure%20%28by%20ref%29))
  (bind _asb_demon_hunter_crossbow_athena (Utilities|Casting|CastToB_DemonHunter_Crossbow_Athena (Components|GetOwner MeshComp))
    (:then
      (Components|Activation|Activate (Variables|/Game/Weapons/FORT_Crossbows/Blueprints/B_DemonHunter_Crossbow_Athena.B_DemonHunter_Crossbow_Athena_C|GetP_DemonHunter_Crossbow_FireTipFlare _asb_demon_hunter_crossbow_athena) false))
    (:CastFailed
      _))
  (return false))
