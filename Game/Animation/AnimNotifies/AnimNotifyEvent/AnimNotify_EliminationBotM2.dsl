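; AnimNotify_EliminationBotM2 is an anim notify that triggers elimination bot VFX from a teleportation drone's animation.
; On notify, it casts the mesh component's owner to BP_TeleportationDrone and takes that drone's owner.
; If that owner implements EliminationBotVFXInterface, it sends the TriggerVFX interface message and returns true.
; A failed cast or missing interface skips the effect.
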(fn Received_Notify
  (params (MeshComp :type Skeletal%20Mesh%20Component%20Object%20Reference) (Animation :type Anim%20Sequence%20Base%20Object%20Reference) (EventReference :type Anim%20Notify%20Event%20Reference%20Structure%20%28by%20ref%29))
  (bind _asbp_teleportation_drone (Utilities|Casting|CastToBP_TeleportationDrone (Components|GetOwner MeshComp))
    (:then
      (if (Utilities|DoesObjectImplementInterface (Actor|GetOwner _asbp_teleportation_drone) "/Game/GameplayCueNotifies/Death/EliminationBotVFXInterface.EliminationBotVFXInterface_C")
        (Class|EliminationBotVFXInterface|TriggerVFX(Message) (Actor|GetOwner _asbp_teleportation_drone))
        (return true)
        (else
          _)))
    (:CastFailed
      _))
  (return false))
