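; An AnimNotify that sends a gameplay event to the owning actor when an animation reaches it, bridging montages or sequences to the Gameplay Ability System.
; Received_Notify gets the mesh component's owner, checks it is valid, then sends the configured EventTag and EventPayload, returning whether the owner was valid.
; GetNotifyName shows the event tag's debug string as the notify's label in the animation editor.
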(fn GetNotifyName
  (return (GameplayTags|GetDebugStringfromGameplayTag (Variables|Default|GetEventTag))))

(fn Received_Notify
  (params (MeshComp :type Skeletal%20Mesh%20Component%20Object%20Reference) (Animation :type Anim%20Sequence%20Base%20Object%20Reference) (EventReference :type Anim%20Notify%20Event%20Reference%20Structure%20%28by%20ref%29))
  (bind _returnvalue (Components|GetOwner MeshComp))
  (Utilities|IsValid _returnvalue
    (:Is%20Valid
      (Ability|SendGameplayEventtoActor _returnvalue (Variables|Default|GetEventTag) (Variables|Default|GetEventPayload))
      (return true))
    (:Is%20Not%20Valid
      (return false))))
