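; Anim notify that switches on melee animation trails for the player pawn during an animation.
; When cosmetic events can execute, it casts the mesh component's owner to FortPlayerPawn and calls AnimTrailsNotify with true.
; It does nothing if the cast fails and returns false after the cosmetic branch.
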(fn Received_Notify
  (params (MeshComp :type Skeletal%20Mesh%20Component%20Object%20Reference) (Animation :type Anim%20Sequence%20Base%20Object%20Reference) (EventReference :type Anim%20Notify%20Event%20Reference%20Structure%20%28by%20ref%29))
  (Networking|CanExecuteCosmeticEvents
    (:True
      (bind _asfort_player_pawn (Utilities|Casting|CastToFortPlayerPawn (Components|GetOwner MeshComp))
        (:then
          (Animation|AnimTrailsNotify _asfort_player_pawn true))
        (:CastFailed
          _))
      (return false))))
