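; Animation notify that plays the clap emote sound on the animated skeletal mesh.
; It casts the mesh component's owner to a Fort player pawn and checks whether that object implements FortVehicleInterface.
; The check selects between the Ostrich clap cue for vehicle implementers and the standard clap cue.
; The chosen cue spawns attached to the mesh component, and the notify returns true.
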(fn Received_Notify
  (params (MeshComp :type Skeletal%20Mesh%20Component%20Object%20Reference) (Animation :type Anim%20Sequence%20Base%20Object%20Reference) (EventReference :type Anim%20Notify%20Event%20Reference%20Structure%20%28by%20ref%29))
  (bind _asfort_player_pawn (Utilities|Casting|CastToFortPlayerPawn (Components|GetOwner MeshComp)))
  (Audio|SpawnSoundAttached (select (Utilities|DoesObjectImplementInterface _asfort_player_pawn "/Script/FortniteGame.FortVehicleInterface") "/Game/Athena/Sounds/Emotes/Athena_Emote_Clap_Ostrich_Cue.Athena_Emote_Clap_Ostrich_Cue" "/Game/Athena/Sounds/Emotes/Athena_Emote_Clap_Cue.Athena_Emote_Clap_Cue") MeshComp)
  (return true))
