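; Anim notify that plays a controller force feedback effect when an animation reaches it, provided a feedback template is set.
; With the world-location option on, it plays near and far templates locally at the mesh's location using inner and outer radii.
; Otherwise it casts the mesh owner to a Fort pawn and plays the template only on that pawn's Fort player controller, if locally controlled.
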(fn Received_Notify
  (params (MeshComp :type Skeletal%20Mesh%20Component%20Object%20Reference) (Animation :type Anim%20Sequence%20Base%20Object%20Reference) (EventReference :type Anim%20Notify%20Event%20Reference%20Structure%20%28by%20ref%29))
  (Utilities|IsValid (Variables|Default|GetForceFeedbackTemplate)
    (:Is%20Valid
      (if (Variables|Default|GetbPlayAtWorldLocation)
        (Fortnite|PlayLocalForceFeedbackatLocationMulti (Variables|Default|GetForceFeedbackTemplate) (Variables|Default|GetForceFeedbackFarTemplate) (Transformation|GetWorldLocation MeshComp) (Variables|Default|GetInnerRadius) (Variables|Default|GetOuterRadius))
        (else
          (bind _asfort_pawn (Utilities|Casting|CastToFortPawn (Components|GetOwner MeshComp))
            (:then
              (bind _asfort_player_controller (Utilities|Casting|CastToFortPlayerController (Pawn|GetController _asfort_pawn))
                (:then
                  (if (Pawn|IsLocalController _asfort_player_controller)
                    (Game|Feedback|ClientPlayForceFeedback _asfort_player_controller (Variables|Default|GetForceFeedbackTemplate) "None" false false false)
                    (return true)
                    (else
                      _)))
                (:CastFailed
                  _)))
            (:CastFailed
              _))
          (return false))))
    (:Is%20Not%20Valid
      _)))
