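; Anim notify condition that decides whether a death sound plays from an animation, judging by its name and the query it makes.
; ShouldTriggerAnimNotify takes the mesh component's owner and looks up the Athena game state from it.
; When both are valid, the game state's ShouldPlayDeathSoundEffects mutator check decides for that owner.
; If either is missing, the notify defaults to triggering.
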(fn ShouldTriggerAnimNotify
  (params (MeshComp :type Skeletal%20Mesh%20Component%20Object%20Reference) (Animation :type Anim%20Sequence%20Base%20Object%20Reference))
  (bind _returnvalue (Components|GetOwner MeshComp))
  (bind _returnvalue_2 (Fortnite|Athena|GameState|GetGameStateAthenafromContext _returnvalue))
  (Utilities|IsValid _returnvalue
    (:Is%20Valid
      (Utilities|IsValid _returnvalue_2
        (:Is%20Valid
          (return (Mutators|ShouldPlayDeathSoundEffects _returnvalue_2 _returnvalue)))
        (:Is%20Not%20Valid
          _)))
    (:Is%20Not%20Valid
      _))
  (return true))
