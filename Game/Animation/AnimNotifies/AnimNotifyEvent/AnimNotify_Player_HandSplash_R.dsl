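; Anim notify that triggers a water hand-splash effect on the right hand of a Fortnite player pawn.
; It casts the mesh owner to the Athena player pawn and invokes the attached splash effect on its mesh at "RightHand".
; It then passes the pawn's previous XY velocity to the active splash Niagara component as a user parameter, printing a debug message if that component cast fails.
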(fn Received_Notify
  (params (MeshComp :type Skeletal%20Mesh%20Component%20Object%20Reference) (Animation :type Anim%20Sequence%20Base%20Object%20Reference) (EventReference :type Anim%20Notify%20Event%20Reference%20Structure%20%28by%20ref%29))
  (bind _asfort_player_pawn_athena (Utilities|Casting|CastToFortPlayerPawnAthena (Components|GetOwner MeshComp))
    (:then
      (FX|InvokeWaterHandSplashAttached _asfort_player_pawn_athena (Variables|/Script/Engine.Character|GetMesh _asfort_player_pawn_athena) "RightHand")
      (bind _asniagara_particle_system_component (Utilities|Casting|CastToNiagaraComponent (Class|FortPlayerPawn|GetActivePlayerHandSplashinWaterComponent _asfort_player_pawn_athena))
        (:then
          (Niagara|SetNiagaraVariable(Vector3) _asniagara_particle_system_component "User.PreviousVelocityXY" (Variables|/Script/FortniteGame.FortPlayerPawnAthena|GetPreviousVelocityXY _asfort_player_pawn_athena)))
        (:CastFailed
          (Development|PrintString (Utilities|String|Append (Utilities|GetDisplayName self) " Cast to Niagara Component Field") true true "(R=1.000000,G=0.285522,B=0.105136,A=1.000000)"))))
    (:CastFailed
      _))
  (return false))
