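; Anim notify that shows or hides the skeletal mesh playing the animation.
; When the notify fires, it checks that the mesh component is valid, then sets its visibility to the inverse of the bHideMesh variable, propagating the change to child components.
; An invalid mesh component is left untouched, and the notify returns false.
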(fn Received_Notify
  (params (MeshComp :type Skeletal%20Mesh%20Component%20Object%20Reference) (Animation :type Anim%20Sequence%20Base%20Object%20Reference) (EventReference :type Anim%20Notify%20Event%20Reference%20Structure%20%28by%20ref%29))
  (Utilities|IsValid MeshComp
    (:Is%20Valid
      (Rendering|SetVisibility MeshComp (Widget|UEFN|InvertBool (Variables|Default|GetbHideMesh)) true))
    (:Is%20Not%20Valid
      _))
  (return false))
