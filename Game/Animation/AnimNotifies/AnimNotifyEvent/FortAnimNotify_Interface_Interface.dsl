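; Blueprint interface for animation notify events that switch a trail effect on and off, judging by its location among the AnimNotify event assets.
; It declares two functions without parameters, TrailOn and TrailOff, and holds no logic of its own.
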(fn TrailOff)

(fn TrailOn)
