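; An anim notify state that puts configurable loose gameplay tags on the animated mesh's owner for the notify window.
; On begin, a random roll against ChanceToAddGameplayTag decides whether the tags are added through the ability system, with replication enabled.
; On end, the same tags are removed with replication, regardless of whether the roll succeeded.
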(fn Received_NotifyBegin
  (params (MeshComp :type Skeletal%20Mesh%20Component%20Object%20Reference) (Animation :type Anim%20Sequence%20Base%20Object%20Reference) (TotalDuration :type Float%20%28single-precision%29) (EventReference :type Anim%20Notify%20Event%20Reference%20Structure%20%28by%20ref%29))
  (if (<= (Math|Random|RandomFloat) (Variables|Default|GetChanceToAddGameplayTag))
    (Ability|GameplayEffect|AddLooseGameplayTags (Components|GetOwner MeshComp) (Variables|Default|GetGameplay%20Tags) true)
    (return true)))

(fn Received_NotifyEnd
  (params (MeshComp :type Skeletal%20Mesh%20Component%20Object%20Reference) (Animation :type Anim%20Sequence%20Base%20Object%20Reference) (EventReference :type Anim%20Notify%20Event%20Reference%20Structure%20%28by%20ref%29))
  (Ability|GameplayEffect|RemoveLooseGameplayTags (Components|GetOwner MeshComp) (Variables|Default|GetGameplay%20Tags) true)
  (return true))
