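; An anim notify state that attaches a configured asset to a socket on the owning FortPawn for the notify's duration.
; On begin, it attaches only if the pawn has none of the DoNotAttach gameplay tags or is locally controlled.
; If an animation is set, it plays it from the start on the attached skeletal mesh, using the configured looping and play rate.
; On end, it detaches the asset without rechecking the tags.
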(fn Received_NotifyBegin
  (params (MeshComp :type Skeletal%20Mesh%20Component%20Object%20Reference) (Animation :type Anim%20Sequence%20Base%20Object%20Reference) (TotalDuration :type Float%20%28single-precision%29) (EventReference :type Anim%20Notify%20Event%20Reference%20Structure%20%28by%20ref%29))
  (Utilities|FlowControl|Sequence
    (:then_0
      (bind _asfort_pawn (Utilities|Casting|CastToFortPawn (Components|GetOwner MeshComp)))
      (if (or (not (GameplayTags|HasAnyMatchingGameplayTags _asfort_pawn (Variables|Default|GetDoNotAttachTagContainer))) (Pawn|IsLocallyControlled _asfort_pawn))
        (Attachments|AttachAssettoSocket _asfort_pawn (Variables|Default|GetAttachment))
        (if (Utilities|IsValid (Variables|Animation|GetAnimation) :!class /Script/Engine.KismetSystemLibrary)
          (bind _returnvalue (Attachments|GetAssetSkeletalMeshComponentAttachment _asfort_pawn (Utilities|Struct|BreakAssetAttachment (Variables|Default|GetAttachment))))
          (if (Utilities|IsValid _returnvalue :!class /Script/Engine.KismetSystemLibrary)
            (Components|Animation|PlayAnimation _returnvalue (Variables|Animation|GetAnimation) (Variables|Animation|GetLooping))
            (Components|Animation|SetPlayRate _returnvalue (Variables|Animation|GetPlayRate))
            (Components|Animation|SetPosition _returnvalue 0.0 true)))))
    (:then_1
      (return false))))

(fn Received_NotifyEnd
  (params (MeshComp :type Skeletal%20Mesh%20Component%20Object%20Reference) (Animation :type Anim%20Sequence%20Base%20Object%20Reference) (EventReference :type Anim%20Notify%20Event%20Reference%20Structure%20%28by%20ref%29))
  (bind _asfort_pawn (Utilities|Casting|CastToFortPawn (Components|GetOwner MeshComp)))
  (Attachments|DetachAssetfromSocket _asfort_pawn (Utilities|Struct|BreakAssetAttachment (Variables|Default|GetAttachment)))
  (return false))
