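; Anim notify state that overrides the player's hand IK while carrying a DBNO (downed) player, adjusting for both characters' builds.
; On begin, if carrying a valid hoistee, it picks left and right hand vectors by Large/Medium body type and gender of both, writing them to Fortnite_M_Avg_Player_AnimBlueprint.
; Medium male carrying medium male uses zero; other carrier body types are left untouched.
; On end, it zeroes both overrides.
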(fn Received_NotifyBegin
  (params (MeshComp :type Skeletal%20Mesh%20Component%20Object%20Reference) (Animation :type Anim%20Sequence%20Base%20Object%20Reference) (TotalDuration :type Float%20%28single-precision%29) (EventReference :type Anim%20Notify%20Event%20Reference%20Structure%20%28by%20ref%29))
  (bind _asfortnite_m_avg_player_anim_blueprint (Utilities|Casting|CastToFortnite_M_Avg_Player_AnimBlueprint (Components|SkeletalMesh|GetAnimInstance MeshComp)))
  (Utilities|FlowControl|Sequence
    (:then_0
      (if (Variables|/Script/FortniteGame.FortPlayerAnimInstance|GetbIsDBNOCarrying _asfortnite_m_avg_player_anim_blueprint)
        (Utilities|IsValid (Class|FortPlayerPawn|GetDBNOHoistee (Variables|/Script/FortniteGame.FortPlayerAnimInstance|GetFortPlayerPawn _asfortnite_m_avg_player_anim_blueprint))
          (:Is%20Valid
            (if (Utilities|Enum|Equal(Enum) (Variables|/Script/FortniteGame.FortPlayerAnimInstance|GetAnimBodyType _asfortnite_m_avg_player_anim_blueprint) "Large")
              (if (Variables|/Script/FortniteGame.FortPlayerAnimInstance|GetbIsFemale _asfortnite_m_avg_player_anim_blueprint)
                (if (Utilities|Enum|Equal(Enum) (Customization|GetCharacterBodyType (Class|FortPlayerPawn|GetDBNOHoistee (Variables|/Script/FortniteGame.FortPlayerAnimInstance|GetFortPlayerPawn _asfortnite_m_avg_player_anim_blueprint))) "Large")
                  (if (Utilities|Enum|Equal(Enum) (Customization|GetCharacterGender (Class|FortPlayerPawn|GetDBNOHoistee (Variables|/Script/FortniteGame.FortPlayerAnimInstance|GetFortPlayerPawn _asfortnite_m_avg_player_anim_blueprint))) "Female")
                    (Variables|/Game/Characters/Player/Male/Male_Avg_Base/Fortnite_M_Avg_Player_AnimBlueprint.Fortnite_M_Avg_Player_AnimBlueprint_C|SetLeftHandOverride (Variables|LargeFemale|GetCLFtoCLFLeftHand) _asfortnite_m_avg_player_anim_blueprint)
                    (Variables|/Game/Characters/Player/Male/Male_Avg_Base/Fortnite_M_Avg_Player_AnimBlueprint.Fortnite_M_Avg_Player_AnimBlueprint_C|SetRightHandOverride (Variables|LargeFemale|GetCLFtoCLFRightHand) _asfortnite_m_avg_player_anim_blueprint)
                    (else
                      (Variables|/Game/Characters/Player/Male/Male_Avg_Base/Fortnite_M_Avg_Player_AnimBlueprint.Fortnite_M_Avg_Player_AnimBlueprint_C|SetLeftHandOverride (Variables|LargeFemale|GetCLFtoCLMLeftHand) _asfortnite_m_avg_player_anim_blueprint)
                      (Variables|/Game/Characters/Player/Male/Male_Avg_Base/Fortnite_M_Avg_Player_AnimBlueprint.Fortnite_M_Avg_Player_AnimBlueprint_C|SetRightHandOverride (Variables|LargeFemale|GetCLFtoCLMRightHand) _asfortnite_m_avg_player_anim_blueprint)))
                  (elif (Utilities|Enum|Equal(Enum) (Customization|GetCharacterGender (Class|FortPlayerPawn|GetDBNOHoistee (Variables|/Script/FortniteGame.FortPlayerAnimInstance|GetFortPlayerPawn _asfortnite_m_avg_player_anim_blueprint))) "Female")
                    (Variables|/Game/Characters/Player/Male/Male_Avg_Base/Fortnite_M_Avg_Player_AnimBlueprint.Fortnite_M_Avg_Player_AnimBlueprint_C|SetLeftHandOverride (Variables|LargeFemale|GetCLFtoCMFLeftHand) _asfortnite_m_avg_player_anim_blueprint)
                    (Variables|/Game/Characters/Player/Male/Male_Avg_Base/Fortnite_M_Avg_Player_AnimBlueprint.Fortnite_M_Avg_Player_AnimBlueprint_C|SetRightHandOverride (Variables|LargeFemale|GetCLFtoCMFRightHand) _asfortnite_m_avg_player_anim_blueprint)
                    (else
                      (Variables|/Game/Characters/Player/Male/Male_Avg_Base/Fortnite_M_Avg_Player_AnimBlueprint.Fortnite_M_Avg_Player_AnimBlueprint_C|SetLeftHandOverride (Variables|LargeFemale|GetCLFtoCMMLeftHand) _asfortnite_m_avg_player_anim_blueprint)
                      (Variables|/Game/Characters/Player/Male/Male_Avg_Base/Fortnite_M_Avg_Player_AnimBlueprint.Fortnite_M_Avg_Player_AnimBlueprint_C|SetRightHandOverride (Variables|LargeFemale|GetCLFtoCMMRightHand) _asfortnite_m_avg_player_anim_blueprint))))
                (elif (Utilities|Enum|Equal(Enum) (Customization|GetCharacterBodyType (Class|FortPlayerPawn|GetDBNOHoistee (Variables|/Script/FortniteGame.FortPlayerAnimInstance|GetFortPlayerPawn _asfortnite_m_avg_player_anim_blueprint))) "Large")
                  (if (Utilities|Enum|Equal(Enum) (Customization|GetCharacterGender (Class|FortPlayerPawn|GetDBNOHoistee (Variables|/Script/FortniteGame.FortPlayerAnimInstance|GetFortPlayerPawn _asfortnite_m_avg_player_anim_blueprint))) "Female")
                    (Variables|/Game/Characters/Player/Male/Male_Avg_Base/Fortnite_M_Avg_Player_AnimBlueprint.Fortnite_M_Avg_Player_AnimBlueprint_C|SetLeftHandOverride (Variables|LargeMale|GetCLMtoCLFLeftHand) _asfortnite_m_avg_player_anim_blueprint)
                    (Variables|/Game/Characters/Player/Male/Male_Avg_Base/Fortnite_M_Avg_Player_AnimBlueprint.Fortnite_M_Avg_Player_AnimBlueprint_C|SetRightHandOverride (Variables|LargeMale|GetCLMtoCLFRightHand) _asfortnite_m_avg_player_anim_blueprint)
                    (else
                      (Variables|/Game/Characters/Player/Male/Male_Avg_Base/Fortnite_M_Avg_Player_AnimBlueprint.Fortnite_M_Avg_Player_AnimBlueprint_C|SetLeftHandOverride (Variables|LargeMale|GetCLMtoCLMLeftHand) _asfortnite_m_avg_player_anim_blueprint)
                      (Variables|/Game/Characters/Player/Male/Male_Avg_Base/Fortnite_M_Avg_Player_AnimBlueprint.Fortnite_M_Avg_Player_AnimBlueprint_C|SetRightHandOverride (Variables|LargeMale|GetCLMtoCLMRightHand) _asfortnite_m_avg_player_anim_blueprint)))
                  (elif (Utilities|Enum|Equal(Enum) (Customization|GetCharacterGender (Class|FortPlayerPawn|GetDBNOHoistee (Variables|/Script/FortniteGame.FortPlayerAnimInstance|GetFortPlayerPawn _asfortnite_m_avg_player_anim_blueprint))) "Female")
                    (Variables|/Game/Characters/Player/Male/Male_Avg_Base/Fortnite_M_Avg_Player_AnimBlueprint.Fortnite_M_Avg_Player_AnimBlueprint_C|SetLeftHandOverride (Variables|LargeMale|GetCLMtoCMFLeftHand) _asfortnite_m_avg_player_anim_blueprint)
                    (Variables|/Game/Characters/Player/Male/Male_Avg_Base/Fortnite_M_Avg_Player_AnimBlueprint.Fortnite_M_Avg_Player_AnimBlueprint_C|SetRightHandOverride (Variables|LargeMale|GetCLMtoCMFRightHand) _asfortnite_m_avg_player_anim_blueprint)
                    (else
                      (Variables|/Game/Characters/Player/Male/Male_Avg_Base/Fortnite_M_Avg_Player_AnimBlueprint.Fortnite_M_Avg_Player_AnimBlueprint_C|SetLeftHandOverride (Variables|LargeMale|GetCLMtoCMMLeftHand) _asfortnite_m_avg_player_anim_blueprint)
                      (Variables|/Game/Characters/Player/Male/Male_Avg_Base/Fortnite_M_Avg_Player_AnimBlueprint.Fortnite_M_Avg_Player_AnimBlueprint_C|SetRightHandOverride (Variables|LargeMale|GetCLMtoCMMRightHand) _asfortnite_m_avg_player_anim_blueprint)))))
              (elif (Utilities|Enum|Equal(Enum) (Variables|/Script/FortniteGame.FortPlayerAnimInstance|GetAnimBodyType _asfortnite_m_avg_player_anim_blueprint) "Medium")
                (if (Variables|/Script/FortniteGame.FortPlayerAnimInstance|GetbIsFemale _asfortnite_m_avg_player_anim_blueprint)
                  (if (Utilities|Enum|Equal(Enum) (Customization|GetCharacterBodyType (Class|FortPlayerPawn|GetDBNOHoistee (Variables|/Script/FortniteGame.FortPlayerAnimInstance|GetFortPlayerPawn _asfortnite_m_avg_player_anim_blueprint))) "Medium")
                    (if (Utilities|Enum|Equal(Enum) (Customization|GetCharacterGender (Class|FortPlayerPawn|GetDBNOHoistee (Variables|/Script/FortniteGame.FortPlayerAnimInstance|GetFortPlayerPawn _asfortnite_m_avg_player_anim_blueprint))) "Female")
                      (Variables|/Game/Characters/Player/Male/Male_Avg_Base/Fortnite_M_Avg_Player_AnimBlueprint.Fortnite_M_Avg_Player_AnimBlueprint_C|SetLeftHandOverride (Variables|MediumFemale|GetCMFtoCMFLeftHand) _asfortnite_m_avg_player_anim_blueprint)
                      (Variables|/Game/Characters/Player/Male/Male_Avg_Base/Fortnite_M_Avg_Player_AnimBlueprint.Fortnite_M_Avg_Player_AnimBlueprint_C|SetRightHandOverride (Variables|MediumFemale|GetCMFtoCMFRightHand) _asfortnite_m_avg_player_anim_blueprint)
                      (else
                        (Variables|/Game/Characters/Player/Male/Male_Avg_Base/Fortnite_M_Avg_Player_AnimBlueprint.Fortnite_M_Avg_Player_AnimBlueprint_C|SetLeftHandOverride (Variables|MediumFemale|GetCMFtoCMMLeftHand) _asfortnite_m_avg_player_anim_blueprint)
                        (Variables|/Game/Characters/Player/Male/Male_Avg_Base/Fortnite_M_Avg_Player_AnimBlueprint.Fortnite_M_Avg_Player_AnimBlueprint_C|SetRightHandOverride (Variables|MediumFemale|GetCMFtoCMMRightHand) _asfortnite_m_avg_player_anim_blueprint)))
                    (elif (Utilities|Enum|Equal(Enum) (Customization|GetCharacterGender (Class|FortPlayerPawn|GetDBNOHoistee (Variables|/Script/FortniteGame.FortPlayerAnimInstance|GetFortPlayerPawn _asfortnite_m_avg_player_anim_blueprint))) "Female")
                      (Variables|/Game/Characters/Player/Male/Male_Avg_Base/Fortnite_M_Avg_Player_AnimBlueprint.Fortnite_M_Avg_Player_AnimBlueprint_C|SetLeftHandOverride (Variables|MediumFemale|GetCMFtoCLFLeftHand) _asfortnite_m_avg_player_anim_blueprint)
                      (Variables|/Game/Characters/Player/Male/Male_Avg_Base/Fortnite_M_Avg_Player_AnimBlueprint.Fortnite_M_Avg_Player_AnimBlueprint_C|SetRightHandOverride (Variables|MediumFemale|GetCMFtoCLFRightHand) _asfortnite_m_avg_player_anim_blueprint)
                      (else
                        (Variables|/Game/Characters/Player/Male/Male_Avg_Base/Fortnite_M_Avg_Player_AnimBlueprint.Fortnite_M_Avg_Player_AnimBlueprint_C|SetLeftHandOverride (Variables|MediumFemale|GetCMFtoCLMLeftHand) _asfortnite_m_avg_player_anim_blueprint)
                        (Variables|/Game/Characters/Player/Male/Male_Avg_Base/Fortnite_M_Avg_Player_AnimBlueprint.Fortnite_M_Avg_Player_AnimBlueprint_C|SetRightHandOverride (Variables|MediumFemale|GetCMFtoCLMRightHand) _asfortnite_m_avg_player_anim_blueprint))))
                  (elif (Utilities|Enum|Equal(Enum) (Customization|GetCharacterBodyType (Class|FortPlayerPawn|GetDBNOHoistee (Variables|/Script/FortniteGame.FortPlayerAnimInstance|GetFortPlayerPawn _asfortnite_m_avg_player_anim_blueprint))) "Medium")
                    (if (Utilities|Enum|Equal(Enum) (Customization|GetCharacterGender (Class|FortPlayerPawn|GetDBNOHoistee (Variables|/Script/FortniteGame.FortPlayerAnimInstance|GetFortPlayerPawn _asfortnite_m_avg_player_anim_blueprint))) "Female")
                      (Variables|/Game/Characters/Player/Male/Male_Avg_Base/Fortnite_M_Avg_Player_AnimBlueprint.Fortnite_M_Avg_Player_AnimBlueprint_C|SetLeftHandOverride (Variables|MediumMale|GetCMMtoCMFLeftHand) _asfortnite_m_avg_player_anim_blueprint)
                      (Variables|/Game/Characters/Player/Male/Male_Avg_Base/Fortnite_M_Avg_Player_AnimBlueprint.Fortnite_M_Avg_Player_AnimBlueprint_C|SetRightHandOverride (Variables|MediumMale|GetCMMtoCMFRightHand) _asfortnite_m_avg_player_anim_blueprint)
                      (else
                        (Variables|/Game/Characters/Player/Male/Male_Avg_Base/Fortnite_M_Avg_Player_AnimBlueprint.Fortnite_M_Avg_Player_AnimBlueprint_C|SetLeftHandOverride "0, 0, 0" _asfortnite_m_avg_player_anim_blueprint)
                        (Variables|/Game/Characters/Player/Male/Male_Avg_Base/Fortnite_M_Avg_Player_AnimBlueprint.Fortnite_M_Avg_Player_AnimBlueprint_C|SetRightHandOverride "0, 0, 0" _asfortnite_m_avg_player_anim_blueprint)))
                    (elif (Utilities|Enum|Equal(Enum) (Customization|GetCharacterGender (Class|FortPlayerPawn|GetDBNOHoistee (Variables|/Script/FortniteGame.FortPlayerAnimInstance|GetFortPlayerPawn _asfortnite_m_avg_player_anim_blueprint))) "Female")
                      (Variables|/Game/Characters/Player/Male/Male_Avg_Base/Fortnite_M_Avg_Player_AnimBlueprint.Fortnite_M_Avg_Player_AnimBlueprint_C|SetLeftHandOverride (Variables|MediumMale|GetCMMtoCLFLeftHand) _asfortnite_m_avg_player_anim_blueprint)
                      (Variables|/Game/Characters/Player/Male/Male_Avg_Base/Fortnite_M_Avg_Player_AnimBlueprint.Fortnite_M_Avg_Player_AnimBlueprint_C|SetRightHandOverride (Variables|MediumMale|GetCMMtoCLFRightHand) _asfortnite_m_avg_player_anim_blueprint)
                      (else
                        (Variables|/Game/Characters/Player/Male/Male_Avg_Base/Fortnite_M_Avg_Player_AnimBlueprint.Fortnite_M_Avg_Player_AnimBlueprint_C|SetLeftHandOverride (Variables|MediumMale|GetCMMtoCLMLeftHand) _asfortnite_m_avg_player_anim_blueprint)
                        (Variables|/Game/Characters/Player/Male/Male_Avg_Base/Fortnite_M_Avg_Player_AnimBlueprint.Fortnite_M_Avg_Player_AnimBlueprint_C|SetRightHandOverride (Variables|MediumMale|GetCMMtoCLMRightHand) _asfortnite_m_avg_player_anim_blueprint)))))))))))
    (:then_1
      (return false))))

(fn Received_NotifyEnd
  (params (MeshComp :type Skeletal%20Mesh%20Component%20Object%20Reference) (Animation :type Anim%20Sequence%20Base%20Object%20Reference) (EventReference :type Anim%20Notify%20Event%20Reference%20Structure%20%28by%20ref%29))
  (bind _asfortnite_m_avg_player_anim_blueprint (Utilities|Casting|CastToFortnite_M_Avg_Player_AnimBlueprint (Components|SkeletalMesh|GetAnimInstance MeshComp)))
  (Variables|/Game/Characters/Player/Male/Male_Avg_Base/Fortnite_M_Avg_Player_AnimBlueprint.Fortnite_M_Avg_Player_AnimBlueprint_C|SetLeftHandOverride "0, 0, 0" _asfortnite_m_avg_player_anim_blueprint)
  (Variables|/Game/Characters/Player/Male/Male_Avg_Base/Fortnite_M_Avg_Player_AnimBlueprint.Fortnite_M_Avg_Player_AnimBlueprint_C|SetRightHandOverride "0, 0, 0" _asfortnite_m_avg_player_anim_blueprint)
  (return false))
