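; Anim notify state that stops the owning pawn from rotating to face the camera while the notify window is active.
; On begin, it gets the mesh owner as a Pawn, casts its controller to FortPlayerController and sets Rotate Pawn to Camera to false.
; On end, it sets the flag back to true rather than restoring the earlier value.
; Pawns without a FortPlayerController are left untouched.
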(fn Received_NotifyBegin
  (params (MeshComp :type Skeletal%20Mesh%20Component%20Object%20Reference) (Animation :type Anim%20Sequence%20Base%20Object%20Reference) (TotalDuration :type Float%20%28single-precision%29) (EventReference :type Anim%20Notify%20Event%20Reference%20Structure%20%28by%20ref%29))
  (bind _aspawn (Utilities|Casting|CastToPawn (Components|GetOwner MeshComp)))
  (bind _asfort_player_controller (Utilities|Casting|CastToFortPlayerController (Pawn|GetController _aspawn))
    (:then
      (Input|SetRotatePawntoCamera _asfort_player_controller false))
    (:CastFailed
      _))
  (return false))

(fn Received_NotifyEnd
  (params (MeshComp :type Skeletal%20Mesh%20Component%20Object%20Reference) (Animation :type Anim%20Sequence%20Base%20Object%20Reference) (EventReference :type Anim%20Notify%20Event%20Reference%20Structure%20%28by%20ref%29))
  (bind _aspawn (Utilities|Casting|CastToPawn (Components|GetOwner MeshComp)))
  (bind _asfort_player_controller (Utilities|Casting|CastToFortPlayerController (Pawn|GetController _aspawn))
    (:then
      (Input|SetRotatePawntoCamera _asfort_player_controller true))
    (:CastFailed
      _))
  (return false))
