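; Niagara anim notify state for the Cereal Box emote that tints the spawned effect to match the player character's cosmetic style.
; On notify begin, it looks up which of three Progressive variant stages is active in the owning FortPlayerPawn's Character loadout slot.
; That index picks a value from a color choice float array, which sets the effect's "Colors" parameter.
; An invalid lookup only prints a failure message.
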(fn Received_NotifyBegin
  (params (MeshComp :type Skeletal%20Mesh%20Component%20Object%20Reference) (Animation :type Anim%20Sequence%20Base%20Object%20Reference) (TotalDuration :type Float%20%28single-precision%29) (EventReference :type Anim%20Notify%20Event%20Reference%20Structure%20%28by%20ref%29))
  (bind _returnvalue (AnimNotify|GetSpawnedEffect MeshComp))
  (LocalVariables|SetNiagara%20System _returnvalue)
  (bind _asfort_player_pawn (Utilities|Casting|CastToFortPlayerPawn (Components|GetOwner MeshComp)))
  (bind (_index _bindexisvalid) (CosmeticLoadout|GetVariantChannelActiveIndexforSlotActor 0 "(TagName=\"LoadoutSlot.Character\")" _asfort_player_pawn "(TagName=\"Cosmetics.Variant.Channel.Progressive\")" (Utilities|Array|MakeArray "(TagName=\"Cosmetics.Variant.Property.Stage1\")" "(TagName=\"Cosmetics.Variant.Property.Stage2\")" "(TagName=\"Cosmetics.Variant.Property.Stage3\")") false false))
  (if _bindexisvalid
    (Effects|Components|ParticleSystem|SetFloatParameter (LocalVariables|GetNiagara%20System) "Colors" (Utilities|Array|Get(acopy) (Variables|Default|GetColor%20Choice%20Float%20Value) _index))
    (else
      (Development|PrintString 0 "Emote Cereal Box Variant Slot Check FAILED")))
  (return false))
