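; Anim notify state that holsters a Fortnite pawn's current weapon during an animation, unless that weapon is a consumable.
; On begin, it casts the mesh owner to FortPawn and holsters only if valid weapon data lacks the Item.Consumable tag.
; On end, it clears the holster without repeating that check.
; Both calls use the animation's name as the holster ID and pass the PlayEquipAnim setting.
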(fn CanApplyHolster
  (params (HolsterId :type Name%20%28by%20ref%29 :default "None"))
  (return true))

(fn Received_NotifyBegin
  (params (MeshComp :type Skeletal%20Mesh%20Component%20Object%20Reference) (Animation :type Anim%20Sequence%20Base%20Object%20Reference) (TotalDuration :type Float%20%28single-precision%29) (EventReference :type Anim%20Notify%20Event%20Reference%20Structure%20%28by%20ref%29))
  (bind _asfort_pawn (Utilities|Casting|CastToFortPawn (Components|GetOwner MeshComp))
    (:then
      (Utilities|IsValid (Variables|/Script/FortniteGame.FortPawn|GetCurrentWeapon _asfort_pawn)
        (:Is%20Valid
          (Utilities|IsValid (Data|GetWeaponData (Variables|/Script/FortniteGame.FortPawn|GetCurrentWeapon _asfort_pawn))
            (:Is%20Valid
              (bind _returnvalue (Fortnite|DoesItemDefinitionHaveGameplayTag (Data|GetWeaponData (Variables|/Script/FortniteGame.FortPawn|GetCurrentWeapon _asfort_pawn)) (GameplayTags|MakeLiteralGameplayTag "(TagName=\"Item.Consumable\")")))
              (if (not _returnvalue)
                (Weapon|SetHolsterWeapon _asfort_pawn 0 true (Variables|Default|GetPlayEquipAnim) (Utilities|String|StringToName (Utilities|GetObjectName Animation)))))
            (:Is%20Not%20Valid
              _)))
        (:Is%20Not%20Valid
          _)))
    (:CastFailed
      _))
  (return false))

(fn Received_NotifyEnd
  (params (MeshComp :type Skeletal%20Mesh%20Component%20Object%20Reference) (Animation :type Anim%20Sequence%20Base%20Object%20Reference) (EventReference :type Anim%20Notify%20Event%20Reference%20Structure%20%28by%20ref%29))
  (bind _asfort_pawn (Utilities|Casting|CastToFortPawn (Components|GetOwner MeshComp))
    (:then
      (Weapon|SetHolsterWeapon _asfort_pawn 0 false (Variables|Default|GetPlayEquipAnim) (Utilities|String|StringToName (Utilities|GetObjectName Animation))))
    (:CastFailed
      _))
  (return false))
