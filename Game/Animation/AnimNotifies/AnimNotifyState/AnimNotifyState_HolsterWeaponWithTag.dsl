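; Anim notify state that, judging by its name, holsters the pawn's weapon during an animation, with a tag-based exception.
; On notify begin it casts the mesh owner to FortPawn and reads the player active granted gameplay tags from the current weapon.
; If any tag in SkipHolsterGameplayTags appears there, it returns false; otherwise it defers to the parent implementation.
; Only the begin event is overridden in this graph.
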(fn Received_NotifyBegin
  (params (MeshComp :type Skeletal%20Mesh%20Component%20Object%20Reference) (Animation :type Anim%20Sequence%20Base%20Object%20Reference) (TotalDuration :type Float%20%28single-precision%29) (EventReference :type Anim%20Notify%20Event%20Reference%20Structure%20%28by%20ref%29))
  (bind _asfort_pawn (Utilities|Casting|CastToFortPawn (Components|GetOwner MeshComp)))
  (for _array_element (GameplayTags|BreakGameplayTagContainer (Variables|Default|GetSkipHolsterGameplayTags))
    (if (Utilities|Array|ContainsItem (GameplayTags|BreakGameplayTagContainer (Tags|GetPlayerActiveGrantedGameplayTags (Variables|/Script/FortniteGame.FortPawn|GetCurrentWeapon _asfort_pawn))) _array_element)
      (return false)))
  (bind _returnvalue (|CallParentFunction MeshComp Animation TotalDuration))
  (return _returnvalue))
