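; Test anim notify state for character leaning that also carries weapon holster logic.
; Tick casts the mesh's anim instance to LeanTest_AnimLayer and calls Lean with zero values.
; Begin only checks a cast to the player AnimBlueprint; both print "fail" on failure.
; End unholsters the owning FortPawn's weapon, passing PlayEquipAnim and the animation name; SetLeanAngle and NewFunction are empty, and CanApplyHolster always returns true.
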(fn CanApplyHolster
  (params (HolsterId :type Name%20%28by%20ref%29 :default "None"))
  (return true))

(fn NewFunction)

(fn Received_NotifyBegin
  (params (MeshComp :type Skeletal%20Mesh%20Component%20Object%20Reference) (Animation :type Anim%20Sequence%20Base%20Object%20Reference) (TotalDuration :type Float%20%28single-precision%29) (EventReference :type Anim%20Notify%20Event%20Reference%20Structure%20%28by%20ref%29))
  (Utilities|Casting|CastToFortnite_M_Avg_Player_AnimBlueprint (Components|SkeletalMesh|GetAnimInstance MeshComp)
    (:CastFailed
      (Development|PrintString 0 "fail"))))

(fn Received_NotifyEnd
  (params (MeshComp :type Skeletal%20Mesh%20Component%20Object%20Reference) (Animation :type Anim%20Sequence%20Base%20Object%20Reference) (EventReference :type Anim%20Notify%20Event%20Reference%20Structure%20%28by%20ref%29))
  (bind _asfort_pawn (Utilities|Casting|CastToFortPawn (Components|GetOwner MeshComp))
    (:then
      ;[1 1 1] FORT-197826 -- If unpossession happens before the end is called, it won't unholster the weapon. Player states exist for any relevant possessed pawn, even on simulated (the role check is what force skips).
      ;
      ; Therefore, if we do not have a player state, we are unpossessed and should force unholster.
      (Weapon|SetHolsterWeapon _asfort_pawn 0 false (Variables|Default|GetPlayEquipAnim) (Utilities|String|StringToName (Utilities|GetObjectName Animation))))
    (:CastFailed
      _))
  (return false))

(fn Received_NotifyTick
  (params (MeshComp :type Skeletal%20Mesh%20Component%20Object%20Reference) (Animation :type Anim%20Sequence%20Base%20Object%20Reference) (FrameDeltaTime :type Float%20%28single-precision%29) (EventReference :type Anim%20Notify%20Event%20Reference%20Structure%20%28by%20ref%29))
  (bind _aslean_test_anim_layer (Utilities|Casting|CastToLeanTest_AnimLayer (Components|SkeletalMesh|GetAnimInstance MeshComp))
    (:then
      (Class|LeanTestAnimLayer|Lean _aslean_test_anim_layer 0 0.0)
      (return false))
    (:CastFailed
      (Development|PrintString 0 "fail"))))

(fn SetLeanAngle)
