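; Anim notify state for mech montages that ends the montage early once the mech starts moving.
; Each tick of the notify window casts the mesh's anim instance to Mech_Veh_AnimBP and the animation to a montage.
; If the owning actor's velocity, converted to actor-local space, has any XY component, it stops that montage with a short blend-out.
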(fn Received_NotifyTick
  (params (MeshComp :type Skeletal%20Mesh%20Component%20Object%20Reference) (Animation :type Anim%20Sequence%20Base%20Object%20Reference) (FrameDeltaTime :type Float%20%28single-precision%29) (EventReference :type Anim%20Notify%20Event%20Reference%20Structure%20%28by%20ref%29))
  (bind _asmech_veh_anim_bp (Utilities|Casting|CastToMech_Veh_AnimBP (Components|SkeletalMesh|GetAnimInstance MeshComp))
    (:then
      (bind _asanim_montage (Utilities|Casting|CastToAnimMontage Animation)
        (:then
          (if (> (Math|Vector|VectorLengthXYSquared (Math|Transform|InverseTransformDirection (Transformation|GetActorTransform (Animation|GetOwningActor _asmech_veh_anim_bp)) (Transformation|GetVelocity (Animation|GetOwningActor _asmech_veh_anim_bp)))) 0.0)
            (Animation|Montage|MontageStop _asmech_veh_anim_bp 0.4 _asanim_montage)
            (else
              _)))
        (:CastFailed
          _)))
    (:CastFailed
      _))
  (return false))
