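; Anim notify state that stops its montage when the owning Character is not moving during the notify window.
; Each tick it checks whether the movement component's last input vector is near zero.
; An optional velocity test compares squared speed against a configurable minimum, avoiding a square root.
; The montage stops on the mesh's anim instance using a configurable blend-out time.
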(fn Received_NotifyTick
  (params (MeshComp :type Skeletal%20Mesh%20Component%20Object%20Reference) (Animation :type Anim%20Sequence%20Base%20Object%20Reference) (FrameDeltaTime :type Float%20%28single-precision%29) (EventReference :type Anim%20Notify%20Event%20Reference%20Structure%20%28by%20ref%29))
  (bind _ascharacter (Utilities|Casting|CastToCharacter (Components|GetOwner MeshComp)))
  (if (not (Utilities|Operators|NotEqual(!=) :A (Pawn|Components|PawnMovement|GetLastInputVector (Variables|/Script/Engine.Character|GetCharacterMovement _ascharacter)) :ErrorTolerance 0.1))
    (Utilities|FlowControl|Sequence
      (:then_0
        _)
      (:then_1
        (if (Variables|Default|GetbIncludeVelocityTest)
          (if (> (Math|Vector|VectorLengthSquared (Transformation|GetVelocity _ascharacter)) (Math|Float|Square (Variables|Default|GetMinVelocityMagnitude)))
            _
            (else
              _)))))
    (bind _asanim_montage (Utilities|Casting|CastToAnimMontage Animation))
    (Animation|Montage|MontageStop (Components|SkeletalMesh|GetAnimInstance MeshComp) (Variables|Default|GetBlend%20Out%20Time) _asanim_montage)
    (return false)))
