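; Anim notify state that cuts a montage short when the owning Fort player pawn starts targeting.
; On each notify tick it casts the mesh owner to FortPlayerPawn and queries IsTargeting through the pawn's animation interface.
; While targeting, it stops the montage playing the notify on the mesh's anim instance, using the Blend Out Time variable as the blend duration.
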(fn Received_NotifyTick
  (params (MeshComp :type Skeletal%20Mesh%20Component%20Object%20Reference) (Animation :type Anim%20Sequence%20Base%20Object%20Reference) (FrameDeltaTime :type Float%20%28single-precision%29) (EventReference :type Anim%20Notify%20Event%20Reference%20Structure%20%28by%20ref%29))
  (bind _asfort_player_pawn (Utilities|Casting|CastToFortPlayerPawn (Components|GetOwner MeshComp)))
  (if (Class|FortPlayerAnimationInterface|IsTargeting (Animation|GetAnimationInterface _asfort_player_pawn))
    (bind _asanim_montage (Utilities|Casting|CastToAnimMontage Animation))
    (Animation|Montage|MontageStop (Components|SkeletalMesh|GetAnimInstance MeshComp) (Variables|Default|GetBlend%20Out%20Time) _asanim_montage)
    (return false)))
