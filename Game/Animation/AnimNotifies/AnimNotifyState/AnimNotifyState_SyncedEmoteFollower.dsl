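; Anim notify state that syncs a follower's emote montage with its group emote partner in Fortnite.
; On notify begin, it casts the mesh owner to a FortPlayerPawn and reads that pawn's group emote tail target.
; If the target is valid, it calls SyncMontages with the target's character mesh first and the notifying mesh second.
; Only the begin event is implemented, with no tick or end logic.
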(fn Received_NotifyBegin
  (params (MeshComp :type Skeletal%20Mesh%20Component%20Object%20Reference) (Animation :type Anim%20Sequence%20Base%20Object%20Reference) (TotalDuration :type Float%20%28single-precision%29) (EventReference :type Anim%20Notify%20Event%20Reference%20Structure%20%28by%20ref%29))
  (bind _asfort_player_pawn (Utilities|Casting|CastToFortPlayerPawn (Components|GetOwner MeshComp)))
  (Utilities|IsValid (Variables|/Script/FortniteGame.FortPlayerPawn|GetGroupEmoteTailTarget _asfort_player_pawn)
    (:Is%20Valid
      (Animation|Utilities|SyncMontages (Variables|/Script/Engine.Character|GetMesh (Variables|/Script/FortniteGame.FortPlayerPawn|GetGroupEmoteTailTarget _asfort_player_pawn)) MeshComp)
      (return false))))
