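; Anim notify state that signals a Victory Drone that its teleport has finished.
; On begin, it casts the mesh owner to BP_VictoryDrone and calls NotifyTeleportFinishedTriggered, returning false if the cast fails.
; The end event does nothing beyond returning true, so only the start of the notify window matters.
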(fn Received_NotifyBegin
  (params (MeshComp :type Skeletal%20Mesh%20Component%20Object%20Reference) (Animation :type Anim%20Sequence%20Base%20Object%20Reference) (TotalDuration :type Float%20%28single-precision%29) (EventReference :type Anim%20Notify%20Event%20Reference%20Structure%20%28by%20ref%29))
  (bind _asbp_victory_drone (Utilities|Casting|CastToBP_VictoryDrone (Components|GetOwner MeshComp))
    (:then
      (Class|BPVictoryDrone|NotifyTeleportFinishedTriggered _asbp_victory_drone)
      (return true))
    (:CastFailed
      (return false))))

(fn Received_NotifyEnd
  (params (MeshComp :type Skeletal%20Mesh%20Component%20Object%20Reference) (Animation :type Anim%20Sequence%20Base%20Object%20Reference) (EventReference :type Anim%20Notify%20Event%20Reference%20Structure%20%28by%20ref%29))
  (return true))
