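; Anim notify state that raises a named boolean anim input event on a Fort pawn for the span of the notify.
; Begin and End cast the mesh component's owner to FortPawn and trigger a bool input event built from the EventName variable, true on begin and default (unset) value on end.
; The notify displays the event name as its own name.
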(fn GetNotifyName
  (return (Utilities|String|ToString(Name) (Variables|Default|GetEventName))))

(fn Received_NotifyBegin
  (params (MeshComp :type Skeletal%20Mesh%20Component%20Object%20Reference) (Animation :type Anim%20Sequence%20Base%20Object%20Reference) (TotalDuration :type Float%20%28single-precision%29) (EventReference :type Anim%20Notify%20Event%20Reference%20Structure%20%28by%20ref%29))
  (bind _asfort_pawn (Utilities|Casting|CastToFortPawn (Components|GetOwner MeshComp)))
  (bind _returnvalue (Class|FortAnimInputEventBool|CreateAnimInputEventBool (Variables|Default|GetEventName) true))
  (Animation|TriggerAnimInputEvent _asfort_pawn _returnvalue)
  (return false))

(fn Received_NotifyEnd
  (params (MeshComp :type Skeletal%20Mesh%20Component%20Object%20Reference) (Animation :type Anim%20Sequence%20Base%20Object%20Reference) (EventReference :type Anim%20Notify%20Event%20Reference%20Structure%20%28by%20ref%29))
  (bind _asfort_pawn (Utilities|Casting|CastToFortPawn (Components|GetOwner MeshComp)))
  (bind _returnvalue (Class|FortAnimInputEventBool|CreateAnimInputEventBool (Variables|Default|GetEventName)))
  (Animation|TriggerAnimInputEvent _asfort_pawn _returnvalue)
  (return false))
