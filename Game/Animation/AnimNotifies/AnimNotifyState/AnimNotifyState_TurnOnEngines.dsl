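; AnimNotifyState_TurnOnEngines is an animation notify state that flags a glider as fully deployed.
; On notify begin it casts the mesh's anim instance to FortGliderAnimInstance and calls SetGliderFullyDeployed with true, returning whether that instance is valid.
; Notify end has no logic, so the state does not clear the flag when its window closes.
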(fn Received_NotifyBegin
  (params (MeshComp :type Skeletal%20Mesh%20Component%20Object%20Reference) (Animation :type Anim%20Sequence%20Base%20Object%20Reference) (TotalDuration :type Float%20%28single-precision%29) (EventReference :type Anim%20Notify%20Event%20Reference%20Structure%20%28by%20ref%29))
  (bind _asfort_glider_anim_instance (Utilities|Casting|CastToFortGliderAnimInstance (Components|SkeletalMesh|GetAnimInstance MeshComp))
    (:then
      (Class|FortGliderAnimInstance|SetGliderFullyDeployed _asfort_glider_anim_instance true))
    (:CastFailed
      _))
  (return (Utilities|IsValid _asfort_glider_anim_instance :!class /Script/Engine.KismetSystemLibrary)))

(fn Received_NotifyEnd
  (params (MeshComp :type Skeletal%20Mesh%20Component%20Object%20Reference) (Animation :type Anim%20Sequence%20Base%20Object%20Reference) (EventReference :type Anim%20Notify%20Event%20Reference%20Structure%20%28by%20ref%29)))
