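; Fortnite projectile trajectory preview actor that draws a spline arc with a target marker at its end, shown only to the instigator in splitscreen.
; While updating is enabled, Tick asks the owner to refresh it through the trajectory ability interface.
; New spline points feed locations into valid and invalid dynamic materials created at construction and move the target.
; Validity changes swap spline mesh materials and toggle valid and invalid target visibility.
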(event EventBeginPlay
  (Utilities|FlowControl|Sequence
    (:then_0
      ;[0.1612 1 0.1056] Initialize parameters on the valid and invalid spline materials.
      (Rendering|Material|SetScalarParameterValue (Variables|VFX|GetSplineMID) "StartDistance" (Variables|Default|GetSplineFadeStartDistance) :!class /Script/Engine.MaterialInstanceDynamic)
      (Rendering|Material|SetScalarParameterValue (Variables|VFX|GetSplineMID) "FadeDistance" (Variables|Default|GetSplineFadeDistance) :!class /Script/Engine.MaterialInstanceDynamic))
    (:then_1
      (Rendering|SetVisibility (Variables|Default|GetTarget) false false))
    (:then_2
      ;[1 1 1] Splitscreen fix: Hide projectile trajectory if this is NOT the instigator
      (Fortnite|SetActorHiddenforNonInstigatorLocalPlayerControllers self))))

(event EventTick
  (params (DeltaSeconds :type Float%20%28single-precision%29))
  ;[1 1 1] Tick - Update trajectory
  (|CallParentFunction DeltaSeconds)
  (if (Variables|Default|GetShouldUpdate)
    (CallFunction|UpdatefromTrajectoryOwner)))

(event Custom|PreSetTrajectorySpline
  (params (SplinePoints :type Array%20of%20Vectors) (SplineTangents :type Array%20of%20Vectors) (TrajectoryHitResult :type Hit%20Result%20Structure)))

(event Trajectory|EventSetTrajectorySpline
  (params (SplinePoints :type Array%20of%20Vectors) (SplineTangents :type Array%20of%20Vectors))
  (bind _returnvalue (Utilities|Array|LastIndex SplinePoints))
  (bind _output (Utilities|Array|Get(aref) SplinePoints _returnvalue))
  ;[1 1 1] SetTrajectorySpline
  (|CallParentFunction SplinePoints SplineTangents)
  (if (> (Utilities|Array|Length SplinePoints) 0)
    (Rendering|Material|SetVectorParameterValue (Variables|VFX|GetSplineMID) "StartLocation" (Math|Conversions|ToLinearColor(Vector) (Utilities|Array|Get(aref) SplinePoints)) :!class /Script/Engine.MaterialInstanceDynamic)
    (Rendering|Material|SetVectorParameterValue (Variables|VFX|GetInvalidSplineMID) "FirstHitLocation" (Math|Conversions|ToLinearColor(Vector) (Utilities|Array|Get(aref) SplinePoints 1)) :!class /Script/Engine.MaterialInstanceDynamic)
    (Rendering|Material|SetVectorParameterValue (Variables|VFX|GetSplineMID) "EndLocation" (Math|Conversions|ToLinearColor(Vector) _output) :!class /Script/Engine.MaterialInstanceDynamic)
    (Transformation|SetWorldLocation (Variables|Default|GetTarget) _output false true)
    (CallFunction|UpdateTargetMeshVisibility)))

(event Custom|SetTrajectoryValid
  (params (bIsTrajectoryValid :type Boolean))
  ;[0.9836 0.3241 1] Update the spline and target materials depending on if the trajectory is valid.
  (Variables|Default|SetbWaitFirstTrajectoryValid false)
  (if (not (== bIsTrajectoryValid (Variables|Default|GetbIsTrajectoryValid)))
    (bind _output_get (Variables|Default|SetbIsTrajectoryValid bIsTrajectoryValid))
    (CallFunction|UpdateTargetMeshVisibility)
    (for _array_element (Variables|Trajectory|GetSplineMeshComponents)
      (Rendering|Material|SetMaterial _array_element 0 (Selection|SelectMaterialInstance (Variables|VFX|GetSplineMID) (Variables|VFX|GetInvalidSplineMID) _output_get)))))

(event Custom|UpdateTargetMeshVisibility
  (bind _bistrajectoryvalid (Variables|Default|GetbIsTrajectoryValid))
  ;[0.9836 0.3241 1] Update the spline and target materials depending on if the trajectory is valid.
  (Rendering|SetVisibility (Variables|Default|GetTarget) _bistrajectoryvalid false)
  (Rendering|SetVisibility (Variables|Default|GetInvalidTarget) (not _bistrajectoryvalid) false))

(event EventEndPlay
  (params (EndPlayReason :type EEndPlayReason%20Enum :default "Destroyed"))
  ;[1 1 1] Remove this actor from all player controllers' HiddenActors array
  (Fortnite|SetActorHiddenforNonInstigatorLocalPlayerControllers self false))

(fn SetProjectileBudgetMesh
  (params (ProjectileBudgetMesh :type Skeletal%20Mesh%20Component%20Budgeted%20Object%20Reference)))

(fn SetShouldUpdateFromOwner
  (params (ShouldUpdate :type Boolean) (Owner :type Object%20Reference))
  (bind _output_get (Variables|Default|SetShouldUpdate ShouldUpdate))
  (if _output_get
    (Variables|Default|SetTrajectoryOwner Owner)
    (else
      (Variables|Default|SetTrajectoryOwner 0))))

(fn UpdateFromTrajectoryOwner
  (bind _asfort_trajectory_ability_interface (Utilities|Casting|CastToFortTrajectoryAbilityInterface (Variables|Default|GetTrajectoryOwner)))
  (TrajectoryInterface|UpdateTrajectory _asfort_trajectory_ability_interface))

(fn ConstructionScript
  ;[0.05805 0.4849 1] Create and store valid and invalid material instances.
  (bind _returnvalue (Rendering|Material|CreateDynamicMaterialInstance (Variables|Default|GetParentMaterial1) "None" "None" :!class /Script/Engine.KismetMaterialLibrary))
  (Variables|VFX|SetSplineMID _returnvalue)
  (bind _returnvalue_2 (Rendering|Material|CreateDynamicMaterialInstance (Variables|Default|GetParentMaterial2) "None" "None" :!class /Script/Engine.KismetMaterialLibrary))
  (Variables|VFX|SetInvalidSplineMID _returnvalue_2))
