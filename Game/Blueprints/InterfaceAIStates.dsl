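; Blueprint interface that declares AI state messages for Husk and Taker enemies, with no implementation.
; Taker functions cover swoop and soul suck attacks, including sound, interruption, attack verification, FSM hit results and special attack timer resets.
; Portal functions cover building-actor add, get and clear, collision, FX, and wall-portal availability and behavior tied to a behavior tree task.
; Remaining messages signal Husk evades and mantle start and end, with path points and wall-type flags.
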(fn BeginTakerSwoopAttack)

(fn EndTakerSwoopAttack)

(fn HuskEvadeStart)

(fn IsTakerWallPortalAvailable
  (params (Portal%20Available? :type Boolean)))

(fn MantleEnd)

(fn MantleStart
  (params (StartPos :type Vector) (MidPos :type Vector) (EndPos :type Vector) (LowWall :type Boolean) (MidWall :type Boolean) (FullWall :type Boolean) (WindowWall :type Boolean)))

(fn PortalAdd
  (params (BuildingActor :type Building%20Actor%20Object%20Reference) (BuildingActorList :type Array%20of%20Building%20Actor%20Object%20References)))

(fn PortalClear)

(fn PortalCollision
  (params (Collide? :type Boolean)))

(fn PortalFX)

(fn PortalGet
  (params (BuildingActor :type Building%20Actor%20Object%20Reference) (BuildingActorList :type Array%20of%20Building%20Actor%20Object%20References) (BT :type Actor%20Object%20Reference)))

(fn ResetTakerSpecialAttackTimer
  (params (Swoop :type Boolean) (SoulSuckMelee :type Boolean) (Portal :type Boolean)))

(fn TakerAttackFSM
  (params (SuccessfulHit :type Boolean)))

(fn TakerSoulSuckInterrupted)

(fn TakerSoulSuckSoundBegin
  (params (Success :type Boolean)))

(fn TakerSoulSuckSoundEnd)

(fn TakerWallPortalBehavior
  (params (BuildingPart :type Actor%20Object%20Reference) (Task :type BTTask%20Blueprint%20Base%20Object%20Reference) (DesiredMoveLocation :type Vector) (Debug :type Boolean)))

(fn VerifyTakerAttack
  (params (SpecialAttack :type Boolean) (Swoop :type Boolean) (Soul%20Suck :type Boolean) (Portal :type Boolean)))
