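; InterfacePlayerPawn declares player pawn functions for water state and two-sided melee swings, and appears to be a Blueprint Interface.
; GetInWater returns a boolean and Melee_Effect_Color returns a color-like triple, with false and black as the stub values.
; Left and right melee swings each have a start function with a first-swing flag and a matching end function.
; All functions are stubs without logic.
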(fn GetInWater
  (return false))

(fn Melee_Effect_Color
  (return "0, 0, 0"))

(fn MeleeSwingLeft
  (params (First%20Left :type Boolean)))

(fn MeleeSwingLeft_End)

(fn MeleeSwingRight
  (params (First%20Right :type Boolean)))

(fn MeleeSwingRight_End)
