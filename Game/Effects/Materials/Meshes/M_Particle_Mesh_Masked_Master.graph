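# Two-sided, masked Default Lit master material for mesh, sprite and ribbon particles.
# Scalar-parameter lerps toggle features: SubUV maps, color-chart tinting, masked Fresnel emissive, jiggle or noise WPO and dithered opacity.
# Material attributes pass through MF_CreativeMoveTool, two optional erosion stages, then MF_FX_MAEmissiveAdjustments.
# Advanced erosion combines bounds ramp, cracks, noise and dynamic Erode to cut opacity, darken edges and push vertices along normals; simple erosion thresholds noise by particle alpha.
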Material /Game/Effects/Materials/Meshes/M_Particle_Mesh_Masked_Master
graph {
 properties {
  bUseMaterialAttributes=true
  bUsedWithMeshParticles=true
  bUsedWithNanite=true
  bUsedWithNiagaraMeshParticles=true
  bUsedWithNiagaraRibbons=true
  bUsedWithNiagaraSprites=true
  bUsedWithParticleSprites=true
  blendMode=BLEND_Masked
  opacityMaskClipValue=0.16
  pixelDepthOffsetMode=PDOM_Legacy
  translucencyLightingMode=TLM_VolumetricPerVertexDirectional
  twoSided=true
  usedShadingModels=Default Lit
 }
 nodes [
  n1 = NamedRerouteUsage()
  n2 = ScalarParameter() {defaultValue=1, group=Base_OpacityMask, parameterName=Diffuse Alpha Blend, sortPriority=1}
  n3 = VectorParameter() {group=Base_OpacityMask, parameterName=Opacity Mask UV Scale, sortPriority=4, defaultValue=<rgba 1 1 0 0>}
  n4 = Multiply(A=TextureCoordinate(), B=ComponentMask(None=@n3.RGB) {g=true, r=true})
  n5 = TextureSampleParameter2D(UVs=LinearInterpolate(
    A=@n4
    B=MaterialFunctionCall(XY=@n4) {materialFunction=&Swizzle}.YX
    Alpha=ScalarParameter() {group=Base_OpacityMask, parameterName=Swizzle Opacity Mask UVs, sortPriority=5}
   )) {
   group=Base_OpacityMask
   parameterName=Opacity_Mask_Texture
   samplerSource=SSM_Wrap_WorldGroupSettings
   sortPriority=2
   texture=&T_WhitesRGB
  }
  n6 = Multiply(
   A=Multiply(A=ParticleColor().A, B=LinearInterpolate(B=@n1, Alpha=@n2) {constA=1})
   B=ChannelMaskParameter(None=@n5.RGBA) {group=Base_OpacityMask, parameterName=Opacity_Mask Channel Mask, sortPriority=3}.R
  )
  n7 = MaterialFunctionCall(Alpha Threshold=@n6) {materialFunction=&DitherTemporalAA}
  n8 = ScalarParameter() {parameterName=UseTemporalAA}
  n9 = NamedRerouteUsage()
  n10 = VectorParameter() {
   group=Base_Normal
   parameterName=Normal Tile
   sortPriority=4
   channelNames {a=SubUV normal Y, b=SubUV normal X, g=Normal Y, r=Normal X}
   defaultValue=<rgba 1 1 1 1>
  }
  n11 = AppendVector(A=@n10.Normal X, B=@n10.Normal X)
  n12 = ScalarParameter() {group=Base_UVs, parameterName=U Speed, sortPriority=1}
  n13 = ScalarParameter() {group=Base_UVs, parameterName=V Speed, sortPriority=2}
  n14 = TextureSampleParameter2D(UVs=Panner(Coordinate=Multiply(A=@n9, B=@n11), Speed=AppendVector(A=@n12, B=@n13))) {
   group=Base_Normal
   parameterName=Normals
   samplerSource=SSM_Wrap_WorldGroupSettings
   samplerType=SAMPLERTYPE_Normal
   sortPriority=1
   texture=/Engine/EngineMaterials/FlatNormal
  }
  n15 = TextureSampleParameterSubUV!(UVs=Multiply(A=@n9, B=AppendVector(A=@n10.SubUV normal X, B=@n10.SubUV normal X)), Apply View MipBias) {group=Base_Normal, parameterName=Normal_SubUV, samplerType=SAMPLERTYPE_Normal, sortPriority=3, texture=&FlatNormal}
  n16 = LinearInterpolate(A=@n14.RGB, B=@n15.RGB, Alpha=ScalarParameter() {group=Base_Normal, parameterName=Use SubUV Normal Texture, sortPriority=2}) {constAlpha=0}
  n17 = ScalarParameter() {group=Base_Normal, parameterName=Normal Flatness, sortPriority=5}
  n18 = MaterialFunctionCall(MA In=MakeMaterialAttributes!(
    BaseColor=NamedRerouteUsage()
    Metallic=NamedRerouteUsage()
    Specular=NamedRerouteUsage()
    Roughness=NamedRerouteUsage()
    Anisotropy
    EmissiveColor=NamedRerouteUsage()
    Opacity
    OpacityMask=LinearInterpolate(A=@n6, B=@n7.Result, Alpha=@n8)
    Normal=MaterialFunctionCall(Normal=@n16, Flatness=@n17) {materialFunction=&FlattenNormal}.Result
    Tangent
    WorldPositionOffset=NamedRerouteUsage()
    SubsurfaceColor
    ClearCoat
    ClearCoatRoughness
    AmbientOcclusion
    Refraction
    CustomizedUVs_0
    CustomizedUVs_1
    CustomizedUVs_2
    CustomizedUVs_3
    CustomizedUVs_4
    CustomizedUVs_5
    CustomizedUVs_6
    CustomizedUVs_7
    PixelDepthOffset
    ShadingModel
    Displacement
   )) {materialFunction=&MF_CreativeMoveTool}
  n19 = GetMaterialAttributes(None=@n18.MA Out) {attributeGetTypes [
    #1
    #2
    #3
    #4
    #5
   ]}
  n20 = VectorParameter() {group=Erosion_Advanced, parameterName=Crack Color, sortPriority=14, defaultValue=<rgba 0.15 0.15 0.15 0>}
  n21 = LinearInterpolate(
   A=TextureCoordinate()
   B=TextureCoordinate() {coordinateIndex=1}
   Alpha=ScalarParameter() {defaultValue=1, group=Erosion_Advanced, parameterName=Use UV1 Crack Texture, sortPriority=10}
  )
  n22 = TextureObjectParameter() {group=Erosion_Advanced, parameterName=Crack, sortPriority=9, texture=/Game/Global/Textures/Default/Masks_and_Gradients/Cracks_02}
  n23 = ScalarParameter() {defaultValue=350, group=Erosion_Advanced, parameterName=Crack World Aligned Texture Size, sortPriority=12}
  n24 = WorldPosition() {worldPositionShaderOffset=WPT_ExcludeAllShaderOffsets}
  n25 = Multiply(
   A=LinearInterpolate(
    A=TextureSample(UVs=@n21, Tex=@n22) {texture=&Cracks_02}.RGB
    B=MaterialFunctionCall(TextureObject=@n22, TextureSize=@n23, WorldPosition=@n24.XYZ) {materialFunction=&WorldAlignedTexture}.XYZ Texture
    Alpha=ScalarParameter() {defaultValue=1, group=Erosion_Advanced, parameterName=Use WorldAligned Crack Texture, sortPriority=11}
   )
   B=ScalarParameter() {defaultValue=1, group=Erosion_Advanced, parameterName=CrackSize, sortPriority=13}
  )
  n26 = ComponentMask(None=@n25) {r=true}
  n27 = Add(A=ObjectRadius()) {constB=0.001}
  n28 = Add(
   A=Subtract(A=WorldPosition() {worldPositionShaderOffset=WPT_ExcludeAllShaderOffsets}.XYZ, B=ObjectPositionWS()) {desc=Doesn't break with different locations}
   B=@n27
  )
  n29 = Multiply(A=@n27) {constB=2}
  n30 = Saturate(None=ComponentMask(None=Divide(A=@n28, B=@n29)) {b=true})
  n31 = LinearInterpolate(
   A=Constant() {r=0.5}
   B=LinearInterpolate(
    A=OneMinus(None=@n30)
    B=@n30
    Alpha=ScalarParameter() {defaultValue=1, group=Erosion_Advanced, parameterName=BoundRampDirection, sortPriority=7}
   )
   Alpha=ScalarParameter() {defaultValue=1, group=Erosion_Advanced, parameterName=Use Bounds in Erosion, sortPriority=6}
  )
  n32 = ComponentMask(None=@n25) {b=true}
  n33 = Multiply(A=@n32, B=ScalarParameter() {defaultValue=0.5, group=Erosion_Advanced, parameterName=CrackAffectMask, sortPriority=8}) {constB=0.25}
  n34 = Multiply(
   A=TextureCoordinate() {coordinateIndex=1}
   B=ComponentMask(None=VectorParameter() {group=Erosion_Advanced, parameterName=Noise_UV_Scale, sortPriority=3, defaultValue=<rgba 2 2 0 0>}.RGB) {g=true, r=true}
  )
  n35 = TextureSampleParameter2D(UVs=LinearInterpolate(
    A=@n34
    B=MaterialFunctionCall(XY=@n34) {materialFunction=&Swizzle}.YX
    Alpha=ScalarParameter() {group=Erosion_Advanced, parameterName=Swizzle Erosion Noise, sortPriority=4}
   )) {
   group=Erosion_Advanced
   parameterName=Erosion Noise
   samplerType=SAMPLERTYPE_LinearColor
   sortPriority=2
   texture=/Game/Characters/Player/Male/Medium/Bodies/M_MED_SRDrift/Textures/T_ElectricNoise
  }
  n36 = ScalarParameter() {defaultValue=1, group=Erosion_Advanced, parameterName=NoiseAffectMask, sortPriority=1}
  n37 = Subtract(
   A=Add(A=Subtract(A=@n31, B=@n33), B=Multiply(A=@n35.R, B=@n36))
   B=DynamicParameter() {defaultValue=<rgba 1 1 0 1>, paramNames [Emissive Intensity, ColorChartUV, Erode, WPOScale]}.Erode
  )
  n38 = Multiply(A=@n37) {constB=-1}
  n39 = Saturate(None=Add(A=@n38) {constB=0.5})
  n40 = ScalarParameter() {defaultValue=0.1, group=Erosion_Advanced, parameterName=Crack Specular, sortPriority=15}
  n41 = ScalarParameter() {defaultValue=1, group=Erosion_Advanced, parameterName=Crack Roughness, sortPriority=16}
  n42 = Multiply(
   A=Multiply(A=VertexNormalWS(), B=Saturate(None=@n38))
   B=ScalarParameter() {defaultValue=50, group=Erosion_Advanced, parameterName=Erosion_WPO_Strength, sortPriority=5}
  ) {constB=100}
  n43 = Constant()
  n44 = SetMaterialAttributes!(
   MaterialAttributes=@n19.MaterialAttributes
   Base Color=LinearInterpolate(A=LinearInterpolate(A=@n19.Base Color, B=@n20.RGB, Alpha=@n26), Alpha=@n39) {constB=0.1}
   Specular=LinearInterpolate(A=LinearInterpolate(A=@n19.Specular, B=@n40, Alpha=@n32), Alpha=@n39) {constB=0.15}
   Roughness=LinearInterpolate(A=LinearInterpolate(A=@n19.Roughness, B=@n41, Alpha=@n32), Alpha=@n39)
   Opacity Mask=Multiply(A=@n19.Opacity Mask, B=@n37)
   World Position Offset=Add(A=@n19.World Position Offset, B=ShadingPathSwitch(Default=@n42, Mobile=@n43))
  ) {attributeSetTypes [
    #1
    #2
    #3
    #4
    #5
   ]}
  n45 = BlendMaterialAttributes(A=@n18.MA Out, B=@n44, Alpha=ScalarParameter() {group=Erosion_Advanced, parameterName=Use Advanced Erosion, sortPriority=0})
  n46 = GetMaterialAttributes(None=@n45) {attributeGetTypes [#4]}
  n47 = Multiply(A=@n46.Opacity Mask, B=NamedRerouteUsage())
  MaterialFunctionCall!(In=BlendMaterialAttributes(
    A=@n45
    B=SetMaterialAttributes!(MaterialAttributes=@n46.MaterialAttributes, Opacity Mask=@n47) {attributeSetTypes [#4]}
    Alpha=ScalarParameter() {group=Erosion_Simple, parameterName=Use Simple Erosion, sortPriority=0}
   )) {materialFunction=/Game/Effects/Niagara/MasterMaterials/Functions/MF_FX_MAEmissiveAdjustments}
  n48 = ScalarParameter() {defaultValue=1, group=Erosion_Simple, parameterName=Erosion UV Scale U, sortPriority=2}
  n49 = ScalarParameter() {defaultValue=1, group=Erosion_Simple, parameterName=Erosion UV Scale V, sortPriority=3}
  n50 = TextureSampleParameter2D(UVs=Multiply(A=TextureCoordinate(), B=AppendVector(A=@n48, B=@n49))) {
   group=Erosion_Simple
   parameterName=ErosionTextureMask
   samplerType=SAMPLERTYPE_LinearColor
   sortPriority=1
   texture=/Game/Global/Textures/Default/Noise/Noise32x32
  }
  n51 = Subtract(A=Add(A=@n50.R), B=LinearInterpolate(Alpha=ParticleColor().A) {constA=1.2, constB=0})
  NamedRerouteDeclaration(None=Clamp(None=Floor(None=@n51))) {name=Simple Erosion, nodeColor=<rgba 0 1 0.9766 1>}
  n52 = CameraVectorWS()
  n53 = AtmosphericLightVector()
  n54 = ScalarParameter() {group=Emissive, parameterName=Emissive Mask Fresnel Fake Slerp, sortPriority=11}
  n55 = LinearInterpolate(
   B=MaterialFunctionCall(
    Camera Vector=LinearInterpolate(A=@n52, B=@n53, Alpha=@n54)
    Power=ScalarParameter() {defaultValue=1, group=Emissive, parameterName=Emissive Mask Fresnel Exponent, sortPriority=12}
    Cheap contrast dark=ScalarParameter() {defaultValue=0.04, group=Emissive, parameterName=Emissive Mask Fresnel Min Value, sortPriority=13}
    Cheap contrast bright=ScalarParameter() {defaultValue=1, group=Emissive, parameterName=Emissive Mask Fresnel Max Value, sortPriority=14}
   ) {materialFunction=&Fresnel_Function}.Result
   Alpha=ScalarParameter() {group=Emissive, parameterName=Emissive Mask Fresnel, sortPriority=10}
  ) {constA=1}
  n56 = ScalarParameter() {
   desc=This option is there if you are using Emissive Mask Fresnel and you like to avoid using the constant value of 1 for lower end platform in replacement to the fresnel.
   group=Emissive
   parameterName=Disable Emissive Mask Fresnel Constant on Low End Platforms
   sortPriority=15
  }
  n57 = Constant()
  n58 = If(A=@n56, "A > B"=@n57, "A == B"=@n57, "A < B"=Constant() {r=1}) {constB=0.99}
  NamedRerouteDeclaration(None=ShadingPathSwitch(Default=QualitySwitch(Default=@n55, Low=@n58), Mobile=@n58) {desc=Migrated from FeatureLevelSwitch}) {name=Emissive Fresnel Function, nodeColor=<rgba 0 1 0.5078 1>}
  n59 = NamedRerouteUsage()
  n60 = TextureSampleParameter2D(UVs=@n59) {
   group=Emissive
   parameterName=Emissive Mask Texture
   samplerSource=SSM_Wrap_WorldGroupSettings
   sortPriority=3
   texture=&T_WhitesRGB
  }
  n61 = TextureSampleParameterSubUV(UVs=@n59) {
   group=Emissive
   parameterName=Emissive Mask SubUV
   samplerSource=SSM_Wrap_WorldGroupSettings
   sortPriority=6
   texture=&T_WhitesRGB
  }
  n62 = Multiply(
   A=NamedRerouteUsage()
   B=LinearInterpolate(A=@n60.RGBA, B=@n61.RGBA, Alpha=ScalarParameter() {group=Emissive, parameterName=Use SubUV Emissive Mask Texture, sortPriority=5}) {constAlpha=0}
  ) {constA=1}
  n63 = LinearInterpolate(
   A=ChannelMaskParameter(None=@n62) {group=Emissive, parameterName=Emissive Mask Texture Channel, sortPriority=4}.R
   B=ComponentMask(None=@n62) {b=true, g=true, r=true}
   Alpha=ScalarParameter() {group=Emissive, parameterName=Emissive Mask Uses RGB, sortPriority=7}
  )
  n64 = Multiply(
   A=Multiply(
    A=NamedRerouteUsage()
    B=DynamicParameter() {defaultValue=<rgba 1 1 0 1>, paramNames [Emissive Intensity, ColorChartUV, Erode, WPOScale]}.Emissive Intensity
   )
   B=LinearInterpolate(A=@n63, B=OneMinus(None=@n63), Alpha=ScalarParameter() {group=Emissive, parameterName=Invert Emissive Mask, sortPriority=8})
  )
  n65 = ParticleColor()
  n66 = LinearInterpolate(
   A=Multiply(A=@n64, B=@n65.RGB)
   B=Add(A=@n64, B=@n65.RGB)
   Alpha=ScalarParameter() {group=Emissive, parameterName=Particle Emissive Mult Or Add, sortPriority=9}
  )
  n67 = NamedRerouteUsage()
  NamedRerouteDeclaration(None=Multiply(A=Multiply(A=@n66, B=@n67), B=NamedRerouteUsage())) {name=Emissive, nodeColor=<rgba 0 0.1683 0.0931 1>}
  n68 = NamedRerouteUsage()
  n69 = AppendVector(
   A=ScalarParameter() {defaultValue=1, group=Emissive Masking, parameterName=Emissive Mask Multiplier Scale U}
   B=ScalarParameter() {defaultValue=1, group=Emissive Masking, parameterName=Emissive Mask Multiplier Scale V}
  )
  n70 = ScalarParameter() {defaultValue=1, group=Emissive Masking, parameterName=Emissive Mask Multiplier Pan Speed U}
  n71 = ScalarParameter() {defaultValue=1, group=Emissive Masking, parameterName=Emissive Mask Multiplier Pan Speed V}
  n72 = Multiply(A=Time(), B=AppendVector(A=@n70, B=@n71))
  n73 = ScalarParameter() {defaultValue=4, group=Emissive Masking, parameterName=Emissive Mask Offset Steps}
  n74 = AppendVector(
   A=MaterialFunctionCall!(TimeSpeed=@n70, Jump Amount Per Step=@n73, Number of Steps Per Unit of Time=@n73, Time Offset) {materialFunction=/Engine/Functions/Engine_MaterialFunctions02/SteppingPannerTime}.Result
   B=MaterialFunctionCall(TimeSpeed=@n71, Jump Amount Per Step=@n73, Number of Steps Per Unit of Time=@n73) {materialFunction=&SteppingPannerTime}.Result
  )
  n75 = ScalarParameter() {group=Emissive Masking, parameterName=Use Stepped Panner For Emissive Multiplier Mask}
  n76 = TextureSampleParameter2D(UVs=Add(A=Multiply(A=@n68, B=@n69), B=LinearInterpolate(A=@n72, B=@n74, Alpha=@n75))) {
   group=Emissive Masking
   parameterName=Emissive Multiplier Texture
   texture=/Game/ContentCreationTools/ObjectPlacement/Textures/T_BlankWhite_01
  }
  n77 = Power(Base=@n76.RGB, Exp=ScalarParameter() {defaultValue=2, group=Emissive Masking, parameterName=Emissive Mask Multiplier Power})
  n78 = Add(
   A=Add(A=NamedRerouteUsage(), B=Constant2Vector() {g=0.25, r=0.25})
   B=AppendVector(
    A=ScalarParameter() {group=Emissive Masking, parameterName=Emissive Mask Multiplier Offset U}
    B=ScalarParameter() {group=Emissive Masking, parameterName=Emissive Mask Multiplier Offset V}
   )
  )
  n79 = Sine(None=@n78)
  n80 = Constant() {r=-1}
  n81 = Constant() {r=1}
  n82 = ScalarParameter() {group=Emissive Masking, parameterName=Emissive Mask Multiplier Thickness U}
  n83 = Ceil(None=Add(A=ComponentMask(None=@n79) {r=true}, B=LinearInterpolate(A=@n80, B=@n81, Alpha=@n82)))
  n84 = Sine(None=@n78)
  n85 = Constant() {r=-1}
  n86 = Constant() {r=1}
  n87 = ScalarParameter() {group=Emissive Masking, parameterName=Emissive Mask Multiplier Thickness V}
  n88 = Ceil(None=Add(A=ComponentMask(None=@n84) {g=true}, B=LinearInterpolate(A=@n85, B=@n86, Alpha=@n87)))
  n89 = Multiply(A=OneMinus(None=Clamp(None=@n83)), B=OneMinus(None=Clamp(None=@n88)))
  NamedRerouteDeclaration(None=LinearInterpolate(B=Multiply(A=@n77, B=@n89), Alpha=ScalarParameter() {group=Emissive Masking, parameterName=Use Emissive Mask, sortPriority=0}) {constA=1}) {name=Emissive Masking, nodeColor=<rgba 0 1 0.25 1>}
  n90 = ScalarParameter() {group=Emissive Sphere Mask, parameterName=Sphere Mask Low Value, sortPriority=1}
  n91 = ScalarParameter() {defaultValue=1, group=Emissive Sphere Mask, parameterName=Sphere Mask High Value, sortPriority=2}
  n92 = WorldPosition()
  n93 = VectorParameter() {group=Emissive Sphere Mask, parameterName=Location Offset, sortPriority=5, defaultValue=<rgba 0 0 0 1>}
  n94 = Add(A=ActorPositionWS(), B=@n93.RGB)
  n95 = ScalarParameter() {defaultValue=50, group=Emissive Sphere Mask, parameterName=Sphere Mask Radius, sortPriority=3}
  n96 = ScalarParameter() {group=Emissive Sphere Mask, parameterName=Sphere Mask Hardness, sortPriority=4}
  n97 = Clamp(None=Power(
    Base=SphereMask(A=@n92.XYZ, B=@n94, Radius=@n95, Hardness=@n96)
    Exp=ScalarParameter() {group=Emissive Sphere Mask, parameterName=Power, sortPriority=5}
   ) {constExponent=1})
  NamedRerouteDeclaration(None=LinearInterpolate(
    B=LinearInterpolate(A=@n90, B=@n91, Alpha=@n97)
    Alpha=ScalarParameter() {group=Emissive Sphere Mask, parameterName="Use Emissive Sphere Mask (World)", sortPriority=-1}
   ) {constA=1}) {name=Emissive Sphere Mask, nodeColor=<rgba 0.4062 0 1 1>}
  n98 = NamedRerouteUsage()
  n99 = VectorParameter() {
   group=Base_Diffuse
   parameterName=Diffuse Tile
   sortPriority=3
   channelNames {a=SubUV Diff Y, b=SubUV Diff X, g=Diff Y, r=Diff X}
   defaultValue=<rgba 1 1 1 1>
  }
  n100 = AppendVector(A=@n99.Diff X, B=@n99.Diff X)
  n101 = ScalarParameter() {group=Base_UVs, parameterName=U Speed, sortPriority=1}
  n102 = ScalarParameter() {group=Base_UVs, parameterName=V Speed, sortPriority=2}
  n103 = TextureSampleParameter2D(UVs=Panner(Coordinate=Multiply(A=@n98, B=@n100), Speed=AppendVector(A=@n101, B=@n102))) {
   group=Base_Diffuse
   parameterName=Diffuse
   samplerSource=SSM_Wrap_WorldGroupSettings
   sortPriority=0
   texture=/Engine/EngineMaterials/DefaultDiffuse
  }
  n104 = TextureSampleParameterSubUV(UVs=Multiply(A=@n98, B=AppendVector(A=@n99.SubUV Diff X, B=@n99.SubUV Diff X))) {
   group=Base_Diffuse
   parameterName=Diffuse_SubUV
   samplerSource=SSM_Wrap_WorldGroupSettings
   sortPriority=2
   texture=&DefaultDiffuse
  }
  n105 = ScalarParameter() {group=Base_Diffuse, parameterName=Use SubUV Diffuse Texture, sortPriority=1}
  NamedRerouteDeclaration(None=LinearInterpolate(A=@n103.RGB, B=@n104.RGB, Alpha=@n105)) {name=Diffuse RGB, nodeColor=<rgba 0 0.0625 1 1>}
  n106 = VectorParameter() {group=Emissive, parameterName="Emissive Color && Intensity", sortPriority=2}
  n107 = NamedRerouteUsage()
  n108 = VectorParameter() {group=Diffuse Color Picker, parameterName=World Space Color Picker Color Override, sortPriority=5, defaultValue=<rgba 1 1 1 1>}
  n109 = MaterialFunctionCall!(f2, f3=ObjectPositionWS(), f4) {materialFunction=/Game/Packages/Fortnite_UniqueMaterials_1/VectorComponentOps}
  n110 = ScalarParameter() {defaultValue=16.71, group=Diffuse Color Picker, parameterName="Color Change Frequency (WS)", sortPriority=8}
  n111 = DynamicParameter() {defaultValue=<rgba 1 1 0 1>, paramNames [Emissive Intensity, ColorChartUV, Erode, WPOScale]}
  n112 = Constant()
  n113 = TextureSampleParameter2D(UVs=LinearInterpolate(
    A=Divide(A=@n109.f3 Add, B=@n110)
    B=AppendVector(A=@n111.ColorChartUV, B=@n112)
    Alpha=ScalarParameter() {group=Diffuse Color Picker, parameterName=Use Dynamic Parameter for Color Chart, sortPriority=9}
   )) {group=Diffuse Color Picker, parameterName=Color Chart, sortPriority=7, texture=/Game/Global/Textures/Default/ColorSets/CarColors_Interiors}
  n114 = ScalarParameter() {defaultValue=1, group=Diffuse Color Picker, parameterName=Use Color Chart, sortPriority=6}
  n115 = TextureSampleParameter2D() {
   group=Diffuse Color Picker
   parameterName=World Space Color Picker Mask
   samplerSource=SSM_Wrap_WorldGroupSettings
   sortPriority=2
   texture=&T_BlacksRGB
  }
  n116 = ChannelMaskParameter(None=@n115.RGBA) {
   group=Diffuse Color Picker
   maskChannel=Blue
   parameterName=World Space Color Picker Mask Channel
   sortPriority=3
   defaultValue=<rgba 0 0 1 0>
  }
  n117 = LinearInterpolate(
   A=Multiply(A=LinearInterpolate(A=@n108.RGB, B=@n113.RGB, Alpha=@n114), B=NamedRerouteUsage())
   B=NamedRerouteUsage()
   Alpha=LinearInterpolate(
    A=@n116.R
    B=OneMinus(None=@n116.R)
    Alpha=ScalarParameter() {group=Diffuse Color Picker, parameterName=Invert Color Picker Mask, sortPriority=4}
   )
  )
  n118 = ScalarParameter() {group=Diffuse Color Picker, parameterName=Use Color Picker Coloring, sortPriority=1}
  n119 = ScalarParameter() {defaultValue=1, group=Base_Diffuse, parameterName=Diffuse Multiplier, sortPriority=4}
  n120 = VectorParameter() {group=Base_Diffuse, parameterName=Complex Diffuse Multiplier, sortPriority=6, defaultValue=<rgba 1 1 1 1>}
  n121 = ScalarParameter() {group=Base_Diffuse, parameterName=Use Complex Diffuse Multipier, sortPriority=5}
  n122 = MaterialFunctionCall(
   Hue Shift Percentage=ScalarParameter() {group=Base_Diffuse, parameterName=Hue Shift Percentage, sortPriority=6}
   Texture=Multiply(A=LinearInterpolate(A=@n107, B=@n117, Alpha=@n118), B=LinearInterpolate(A=@n119, B=@n120.RGB, Alpha=@n121))
  ) {materialFunction=&HueShift}
  NamedRerouteDeclaration(None=LinearInterpolate(
    A=@n106.RGB
    B=Multiply(A=@n122.Result, B=@n106.RGB)
    Alpha=ScalarParameter() {group=Emissive, parameterName=UseDiffuseinEmissive, sortPriority=1}
   )) {name=Diffuse to Emissive, nodeColor=<rgba 1 0 0.7031 1>}
  n123 = Multiply(A=ParticleColor().RGB, B=@n122.Result)
  n124 = NamedRerouteUsage()
  NamedRerouteDeclaration(None=LinearInterpolate(A=@n123, B=Multiply(A=@n123, B=@n124), Alpha=ScalarParameter() {group=Emissive Sphere Mask, parameterName=Base Color Influence})) {name=Base Color, nodeColor=<rgba 0 1 0.9766 1>}
  NamedRerouteDeclaration(None=LinearInterpolate(
    A=TextureCoordinate()
    B=TextureCoordinate() {coordinateIndex=1}
    Alpha=ScalarParameter() {desc=Applies to all Textures, group=Base_UVs, parameterName=Use Second UV Channel, sortPriority=3}
   )) {name=UVs, nodeColor=<rgba 1 0 0.7969 1>}
  n125 = MaterialFunctionCall(Float3=Subtract(A=WorldPosition().XYZ, B=ObjectPositionWS())) {materialFunction=&BreakOutFloat3Components}
  n126 = Divide(A=@n125.B, B=DynamicParameter() {parameterIndex=1, paramNames [Amplitude,Frequency,Offset,Axis]}.Axis) {constB=100, desc=Falloff}
  n127 = ParticleRandom()
  n128 = DynamicParameter() {parameterIndex=1, paramNames [Amplitude,Frequency,Offset,Axis]}
  n129 = Divide(A=@n125.B, B=@n128.Frequency) {constB=50}
  n130 = ScalarParameter() {group=WorldPositionOffset_Advanced, parameterName=Noise Offset, sortPriority=4}
  n131 = TextureObjectParameter() {
   group=WorldPositionOffset_Advanced
   parameterName=Vertex Offset
   samplerType=SAMPLERTYPE_Normal
   sortPriority=3
   texture=/Game/Effects/Fort_Effects/Textures/T_DustDistortion
  }
  n132 = TextureSample(UVs=Add(A=AppendVector(A=@n127, B=@n129), B=Add(A=@n128.Offset, B=@n130)), Tex=@n131)
  n133 = DynamicParameter() {desc=Axis of influence, parameterIndex=2, paramNames [Amplitude,Frequency,Offset,Axis]}
  n134 = MaterialFunctionCall(X=@n133.Amplitude, Y=@n133.Amplitude, Z=@n133.Amplitude) {materialFunction=&MakeFloat3}
  n135 = Multiply(
   A=Multiply(A=Clamp(None=@n126), B=Multiply(A=@n132.RGB, B=@n134.Result)) {constB=3}
   B=DynamicParameter() {parameterIndex=1, paramNames [Amplitude,Frequency,Offset,Axis]}.Amplitude
  )
  n136 = LinearInterpolate(
   B=LinearInterpolate(
    A=@n135
    B=Transform(None=@n135) {transformSourceType=TRANSFORMSOURCE_Local}
    Alpha=ScalarParameter() {group=WorldPositionOffset_Advanced, parameterName=Is In Local Space, sortPriority=2}
   )
   Alpha=ScalarParameter() {group=WorldPositionOffset_Advanced, parameterName=Use Advanced WPO, sortPriority=0}
  )
  n137 = MaterialFunctionCall!(InVector=@n136) {materialFunction=/Game/Athena/Environments/Landscape/Foliage/MF_MakePreciseVec3}
  n138 = Constant() {r=1}
  n139 = VertexColor()
  n140 = ScalarParameter() {desc=Enables Vertex COlor R to mask WPO., group=WorldPositionOffset_Advanced, parameterName=Vertex Color R Masks WPO, sortPriority=5}
  NamedRerouteDeclaration(None=Multiply(A=ShadingPathSwitch(Default=@n137.Result, Mobile=@n136), B=LinearInterpolate(A=@n138, B=@n139.R, Alpha=@n140))) {name=Advanced WPO, nodeColor=<rgba 1 0.7031 0 1>}
  n141 = MaterialFunctionCall!(
   Speed=ScalarParameter() {defaultValue=0.5, group=WorldPositionOffset, parameterName=Simple_WPO_Speed, sortPriority=1}
   LargeWaveScle=ScalarParameter() {defaultValue=1, group=WorldPositionOffset, parameterName=Simple_WPO_LargeWaveScale, sortPriority=3}
   WaveScale=ScalarParameter() {defaultValue=0.5, group=WorldPositionOffset, parameterName=Simple_WPO_WaveScale, sortPriority=2}
  ) {materialFunction=/Game/Effects/Fort_Effects/Materials/_Functions/Jiggle_WPO_VertColor_Red_3Vec}
  n142 = DynamicParameter() {defaultValue=<rgba 1 1 0 1>, paramNames [Emissive Intensity, ColorChartUV, Erode, WPOScale]}
  NamedRerouteDeclaration(None=LinearInterpolate(
    B=Multiply(A=@n141.Result, B=@n142.WPOScale)
    Alpha=ScalarParameter() {group=WorldPositionOffset, parameterName=Use Simple WPO, sortPriority=0}
   )) {name=Simple WPO, nodeColor=<rgba 0.2422 0 1 1>}
  NamedRerouteDeclaration(None=Add(A=NamedRerouteUsage(), B=NamedRerouteUsage())) {name=WPO, nodeColor=<rgba 0.0781 1 0 1>}
  n143 = NamedRerouteUsage()
  n144 = VectorParameter() {
   group=Base_Specular
   parameterName=Specular Metallic Roughness Mask Tile
   sortPriority=4
   channelNames {a=SubUV Spec Y, b=SubUV Spec X, g=Spec Y, r=Spec X}
   defaultValue=<rgba 1 1 1 1>
  }
  n145 = AppendVector(A=@n144.Spec X, B=@n144.Spec X)
  n146 = ScalarParameter() {group=Base_UVs, parameterName=U Speed, sortPriority=1}
  n147 = ScalarParameter() {group=Base_UVs, parameterName=V Speed, sortPriority=2}
  n148 = TextureSampleParameter2D(UVs=Panner(Coordinate=Multiply(A=@n143, B=@n145), Speed=AppendVector(A=@n146, B=@n147))) {
   desc="R=Specular, G=Metallic, B=Roughness"
   group=Base_Specular
   parameterName=Specular Metallic Roughness Mask
   samplerSource=SSM_Wrap_WorldGroupSettings
   samplerType=SAMPLERTYPE_Masks
   sortPriority=1
   texture=&T_Fortnite_Default_S
  }
  n149 = TextureSampleParameterSubUV(UVs=Multiply(A=@n143, B=AppendVector(A=@n144.SubUV Spec X, B=@n144.SubUV Spec X))) {
   group=Base_Specular
   parameterName=Specular Metallic Roughness Mask_SubUV
   samplerSource=SSM_Wrap_WorldGroupSettings
   samplerType=SAMPLERTYPE_Masks
   sortPriority=3
   texture=&T_Fortnite_Default_S
  }
  n150 = MaterialFunctionCall(Color=LinearInterpolate(
    A=@n148.RGB
    B=@n149.RGB
    Alpha=ScalarParameter() {group=Base_Specular, parameterName=Use SubUV for Specular Metallic Roughness Mask, sortPriority=2}
   ) {constAlpha=0}) {materialFunction=&SplitComponents}
  n151 = LinearInterpolate(
   A=ScalarParameter() {group=Base_Specular, parameterName=SpecRoughnessMin, sortPriority=5}
   B=ScalarParameter() {defaultValue=1, group=Base_Specular, parameterName=SpecRoughnessMax, sortPriority=6}
   Alpha=@n150.B
  )
  n152 = ScalarParameter() {
   desc="Assume this needs to be 1 TRUE, unless working with older Specular Maps"
   group=Base_Specular
   parameterName="Swap Metallic(G->B) and Roughness (B->G)"
   sortPriority=0
  }
  NamedRerouteDeclaration(None=LinearInterpolate(A=@n151, B=@n150.G, Alpha=@n152)) {name="Metallic (G)", nodeColor=<rgba 1 0.1875 0 1>}
  NamedRerouteDeclaration(None=@n150.R) {name="Specular (R)", nodeColor=<rgba 0 0.1328 1 1>}
  NamedRerouteDeclaration(None=LinearInterpolate(A=@n151, B=@n150.G, Alpha=@n152)) {name="Roughness (B)", nodeColor=<rgba 0.4531 1 0 1>}
  NamedRerouteDeclaration(None=LinearInterpolate(A=@n103.A, B=@n104.A, Alpha=@n105)) {name=Diffuse A, nodeColor=<rgba 0 0.1328 1 1>}
 ]
}
