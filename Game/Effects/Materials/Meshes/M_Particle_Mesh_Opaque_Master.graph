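# Opaque, two-sided, lit master material for Niagara and Cascade particle meshes, sprites and ribbons.
# Panning or SubUV maps drive color, normal and specular/metallic/roughness, with color-chart tinting and hue shift.
# Emissive combines masks, fresnel (constant on low-end/mobile) and a world sphere mask; dynamic parameters drive intensity and jiggle or noise-texture WPO.
# Scalar lerps, not static switches, toggle features; material attributes pass through creative move-tool and emissive-adjustment functions.
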Material /Game/Effects/Materials/Meshes/M_Particle_Mesh_Opaque_Master
graph {
 properties {
  bAutomaticallySetUsageInEditor=false
  bUseMaterialAttributes=true
  bUsedWithMeshParticles=true
  bUsedWithNiagaraMeshParticles=true
  bUsedWithNiagaraRibbons=true
  bUsedWithNiagaraSprites=true
  bUsedWithParticleSprites=true
  bUsedWithStaticMesh=false
  opacityMaskClipValue=0.16
  pixelDepthOffsetMode=PDOM_Legacy
  translucencyLightingMode=TLM_VolumetricPerVertexDirectional
  twoSided=true
  usedShadingModels=Default Lit
 }
 nodes [
  n1 = NamedRerouteUsage()
  n2 = NamedRerouteUsage()
  n3 = NamedRerouteUsage()
  n4 = NamedRerouteUsage()
  n5 = NamedRerouteUsage()
  n6 = NamedRerouteUsage()
  n7 = VectorParameter() {
   group=Base_Normal
   parameterName=Normal Tile
   sortPriority=4
   channelNames {a=SubUV normal Y, b=SubUV normal X, g=Normal Y, r=Normal X}
   defaultValue=<rgba 1 1 1 1>
  }
  n8 = AppendVector(A=@n7.Normal X, B=@n7.Normal X)
  n9 = ScalarParameter() {group=Base_UVs, parameterName=U Speed, sortPriority=1}
  n10 = ScalarParameter() {group=Base_UVs, parameterName=V Speed, sortPriority=2}
  n11 = TextureSampleParameter2D(UVs=Panner(Coordinate=Multiply(A=@n6, B=@n8), Speed=AppendVector(A=@n9, B=@n10))) {
   group=Base_Normal
   parameterName=Normals
   samplerSource=SSM_Wrap_WorldGroupSettings
   samplerType=SAMPLERTYPE_Normal
   sortPriority=1
   texture=/Engine/EngineMaterials/FlatNormal
  }
  n12 = TextureSampleParameterSubUV!(UVs=Multiply(A=@n6, B=AppendVector(A=@n7.SubUV normal X, B=@n7.SubUV normal X)), Apply View MipBias) {group=Base_Normal, parameterName=Normal_SubUV, samplerType=SAMPLERTYPE_Normal, sortPriority=3, texture=&FlatNormal}
  n13 = MaterialFunctionCall(
   Normal=LinearInterpolate(A=@n11.RGB, B=@n12.RGB, Alpha=ScalarParameter() {group=Base_Normal, parameterName=Use SubUV Normal Texture, sortPriority=2}) {constAlpha=0}
   Flatness=ScalarParameter() {group=Base_Normal, parameterName=Normal Flatness, sortPriority=5}
  ) {materialFunction=&FlattenNormal}
  n14 = NamedRerouteUsage()
  MaterialFunctionCall!(In=MaterialFunctionCall(MA In=MakeMaterialAttributes!(
     BaseColor=@n1
     Metallic=@n2
     Specular=@n3
     Roughness=@n4
     Anisotropy
     EmissiveColor=@n5
     Opacity
     OpacityMask
     Normal=@n13.Result
     Tangent
     WorldPositionOffset=@n14
     SubsurfaceColor
     ClearCoat
     ClearCoatRoughness
     AmbientOcclusion
     Refraction
     CustomizedUVs_0
     CustomizedUVs_1
     CustomizedUVs_2
     CustomizedUVs_3
     CustomizedUVs_4
     CustomizedUVs_5
     CustomizedUVs_6
     CustomizedUVs_7
     PixelDepthOffset
     ShadingModel
     Displacement
    )) {materialFunction=&MF_CreativeMoveTool}.MA Out) {materialFunction=/Game/Effects/Niagara/MasterMaterials/Functions/MF_FX_MAEmissiveAdjustments}
  n15 = NamedRerouteUsage()
  n16 = VectorParameter() {
   group=Base_Diffuse
   parameterName=Diffuse Tile
   sortPriority=3
   channelNames {a=SubUV Diff Y, b=SubUV Diff X, g=Diff Y, r=Diff X}
   defaultValue=<rgba 1 1 1 1>
  }
  n17 = AppendVector(A=@n16.Diff X, B=@n16.Diff X)
  n18 = ScalarParameter() {group=Base_UVs, parameterName=U Speed, sortPriority=1}
  n19 = ScalarParameter() {group=Base_UVs, parameterName=V Speed, sortPriority=2}
  n20 = TextureSampleParameter2D(UVs=Panner(Coordinate=Multiply(A=@n15, B=@n17), Speed=AppendVector(A=@n18, B=@n19))) {
   group=Base_Diffuse
   parameterName=Diffuse
   samplerSource=SSM_Wrap_WorldGroupSettings
   sortPriority=0
   texture=/Engine/EngineMaterials/DefaultDiffuse
  }
  n21 = TextureSampleParameterSubUV(UVs=Multiply(A=@n15, B=AppendVector(A=@n16.SubUV Diff X, B=@n16.SubUV Diff X))) {
   group=Base_Diffuse
   parameterName=Diffuse_SubUV
   samplerSource=SSM_Wrap_WorldGroupSettings
   sortPriority=2
   texture=&DefaultDiffuse
  }
  NamedRerouteDeclaration(None=LinearInterpolate(A=@n20.RGB, B=@n21.RGB, Alpha=ScalarParameter() {group=Base_Diffuse, parameterName=Use SubUV Diffuse Texture, sortPriority=1})) {name=Diffuse RGB, nodeColor=<rgba 0 0.0625 1 1>}
  NamedRerouteDeclaration(None=LinearInterpolate(
    A=TextureCoordinate()
    B=TextureCoordinate() {coordinateIndex=1}
    Alpha=ScalarParameter() {desc=Applies to all Textures, group=Base_UVs, parameterName=Use Second UV Channel, sortPriority=3}
   )) {name=UVs, nodeColor=<rgba 1 0 0.7969 1>}
  n22 = NamedRerouteUsage()
  n23 = AppendVector(
   A=ScalarParameter() {defaultValue=1, group=Emissive Masking, parameterName=Emissive Mask Multiplier Scale U}
   B=ScalarParameter() {defaultValue=1, group=Emissive Masking, parameterName=Emissive Mask Multiplier Scale V}
  )
  n24 = ScalarParameter() {defaultValue=1, group=Emissive Masking, parameterName=Emissive Mask Multiplier Pan Speed U}
  n25 = ScalarParameter() {defaultValue=1, group=Emissive Masking, parameterName=Emissive Mask Multiplier Pan Speed V}
  n26 = Multiply(A=Time(), B=AppendVector(A=@n24, B=@n25))
  n27 = ScalarParameter() {defaultValue=4, group=Emissive Masking, parameterName=Emissive Mask Offset Steps}
  n28 = AppendVector(
   A=MaterialFunctionCall!(TimeSpeed=@n24, Jump Amount Per Step=@n27, Number of Steps Per Unit of Time=@n27, Time Offset) {materialFunction=/Engine/Functions/Engine_MaterialFunctions02/SteppingPannerTime}.Result
   B=MaterialFunctionCall(TimeSpeed=@n25, Jump Amount Per Step=@n27, Number of Steps Per Unit of Time=@n27) {materialFunction=&SteppingPannerTime}.Result
  )
  n29 = ScalarParameter() {group=Emissive Masking, parameterName=Use Stepped Panner For Emissive Multiplier Mask}
  n30 = TextureSampleParameter2D(UVs=Add(A=Multiply(A=@n22, B=@n23), B=LinearInterpolate(A=@n26, B=@n28, Alpha=@n29))) {
   group=Emissive Masking
   parameterName=Emissive Multiplier Texture
   texture=/Game/ContentCreationTools/ObjectPlacement/Textures/T_BlankWhite_01
  }
  n31 = Power(Base=@n30.RGB, Exp=ScalarParameter() {defaultValue=2, group=Emissive Masking, parameterName=Emissive Mask Multiplier Power})
  n32 = Add(
   A=Add(A=NamedRerouteUsage(), B=Constant2Vector() {g=0.25, r=0.25})
   B=AppendVector(
    A=ScalarParameter() {group=Emissive Masking, parameterName=Emissive Mask Multiplier Offset U}
    B=ScalarParameter() {group=Emissive Masking, parameterName=Emissive Mask Multiplier Offset V}
   )
  )
  n33 = Sine(None=@n32)
  n34 = Constant() {r=-1}
  n35 = Constant() {r=1}
  n36 = ScalarParameter() {group=Emissive Masking, parameterName=Emissive Mask Multiplier Thickness U}
  n37 = Ceil(None=Add(A=ComponentMask(None=@n33) {r=true}, B=LinearInterpolate(A=@n34, B=@n35, Alpha=@n36)))
  n38 = Sine(None=@n32)
  n39 = Constant() {r=-1}
  n40 = Constant() {r=1}
  n41 = ScalarParameter() {group=Emissive Masking, parameterName=Emissive Mask Multiplier Thickness V}
  n42 = Ceil(None=Add(A=ComponentMask(None=@n38) {g=true}, B=LinearInterpolate(A=@n39, B=@n40, Alpha=@n41)))
  n43 = Multiply(A=OneMinus(None=Clamp(None=@n37)), B=OneMinus(None=Clamp(None=@n42)))
  NamedRerouteDeclaration(None=LinearInterpolate(B=Multiply(A=@n31, B=@n43), Alpha=ScalarParameter() {group=Emissive Masking, parameterName=Use Emissive Mask, sortPriority=0}) {constA=1}) {name=Emissive Masking, nodeColor=<rgba 0 1 0.25 1>}
  n44 = ScalarParameter() {group=Emissive Sphere Mask, parameterName=Sphere Mask Low Value, sortPriority=1}
  n45 = ScalarParameter() {defaultValue=1, group=Emissive Sphere Mask, parameterName=Sphere Mask High Value, sortPriority=2}
  n46 = WorldPosition()
  n47 = VectorParameter() {group=Emissive Sphere Mask, parameterName=Location Offset, sortPriority=5, defaultValue=<rgba 0 0 0 1>}
  n48 = Add(A=ActorPositionWS(), B=@n47.RGB)
  n49 = ScalarParameter() {defaultValue=50, group=Emissive Sphere Mask, parameterName=Sphere Mask Radius, sortPriority=3}
  n50 = ScalarParameter() {group=Emissive Sphere Mask, parameterName=Sphere Mask Hardness, sortPriority=4}
  n51 = Clamp(None=Power(
    Base=SphereMask(A=@n46.XYZ, B=@n48, Radius=@n49, Hardness=@n50)
    Exp=ScalarParameter() {group=Emissive Sphere Mask, parameterName=Power, sortPriority=5}
   ) {constExponent=1})
  NamedRerouteDeclaration(None=LinearInterpolate(
    B=LinearInterpolate(A=@n44, B=@n45, Alpha=@n51)
    Alpha=ScalarParameter() {group=Emissive Sphere Mask, parameterName="Use Emissive Sphere Mask (World)", sortPriority=-1}
   ) {constA=1}) {name=Emissive Sphere Mask, nodeColor=<rgba 0.4062 0 1 1>}
  n52 = MaterialFunctionCall!(
   Speed=ScalarParameter() {defaultValue=0.5, group=WorldPositionOffset, parameterName=Simple_WPO_Speed, sortPriority=1}
   LargeWaveScle=ScalarParameter() {defaultValue=1, group=WorldPositionOffset, parameterName=Simple_WPO_LargeWaveScale, sortPriority=3}
   WaveScale=ScalarParameter() {defaultValue=0.5, group=WorldPositionOffset, parameterName=Simple_WPO_WaveScale, sortPriority=2}
  ) {materialFunction=/Game/Effects/Fort_Effects/Materials/_Functions/Jiggle_WPO_VertColor_Red_3Vec}
  n53 = DynamicParameter() {defaultValue=<rgba 1 1 0 1>, paramNames [Emissive Intensity, ColorChartUV, Erode, WPOScale]}
  NamedRerouteDeclaration(None=LinearInterpolate(
    B=Multiply(A=@n52.Result, B=@n53.WPOScale)
    Alpha=ScalarParameter() {group=WorldPositionOffset, parameterName=Use Simple WPO, sortPriority=0}
   )) {name=Simple WPO, nodeColor=<rgba 0.2422 0 1 1>}
  NamedRerouteDeclaration(None=Add(A=NamedRerouteUsage(), B=NamedRerouteUsage())) {name=WPO, nodeColor=<rgba 0.0781 1 0 1>}
  n54 = VectorParameter() {group=Emissive, parameterName="Emissive Color && Intensity", sortPriority=2}
  n55 = NamedRerouteUsage()
  n56 = VectorParameter() {group=Diffuse Color Picker, parameterName=World Space Color Picker Color Override, sortPriority=5, defaultValue=<rgba 1 1 1 1>}
  n57 = MaterialFunctionCall!(f2, f3=ObjectPositionWS(), f4) {materialFunction=/Game/Packages/Fortnite_UniqueMaterials_1/VectorComponentOps}
  n58 = ScalarParameter() {defaultValue=16.71, group=Diffuse Color Picker, parameterName="Color Change Frequency (WS)", sortPriority=8}
  n59 = DynamicParameter() {defaultValue=<rgba 1 1 0 1>, paramNames [Emissive Intensity, ColorChartUV, Erode, WPOScale]}
  n60 = Constant()
  n61 = TextureSampleParameter2D(UVs=LinearInterpolate(
    A=Divide(A=@n57.f3 Add, B=@n58)
    B=AppendVector(A=@n59.ColorChartUV, B=@n60)
    Alpha=ScalarParameter() {group=Diffuse Color Picker, parameterName=Use Dynamic Parameter for Color Chart, sortPriority=9}
   )) {group=Diffuse Color Picker, parameterName=Color Chart, sortPriority=7, texture=/Game/Global/Textures/Default/ColorSets/CarColors_Interiors}
  n62 = ScalarParameter() {defaultValue=1, group=Diffuse Color Picker, parameterName=Use Color Chart, sortPriority=6}
  n63 = TextureSampleParameter2D() {
   group=Diffuse Color Picker
   parameterName=World Space Color Picker Mask
   samplerSource=SSM_Wrap_WorldGroupSettings
   sortPriority=2
   texture=&T_BlacksRGB
  }
  n64 = ChannelMaskParameter(None=@n63.RGBA) {
   group=Diffuse Color Picker
   maskChannel=Blue
   parameterName=World Space Color Picker Mask Channel
   sortPriority=3
   defaultValue=<rgba 0 0 1 0>
  }
  n65 = LinearInterpolate(
   A=Multiply(A=LinearInterpolate(A=@n56.RGB, B=@n61.RGB, Alpha=@n62), B=NamedRerouteUsage())
   B=NamedRerouteUsage()
   Alpha=LinearInterpolate(
    A=@n64.R
    B=OneMinus(None=@n64.R)
    Alpha=ScalarParameter() {group=Diffuse Color Picker, parameterName=Invert Color Picker Mask, sortPriority=4}
   )
  )
  n66 = ScalarParameter() {group=Diffuse Color Picker, parameterName=Use Color Picker Coloring, sortPriority=1}
  n67 = ScalarParameter() {defaultValue=1, group=Base_Diffuse, parameterName=Diffuse Multiplier, sortPriority=4}
  n68 = VectorParameter() {group=Base_Diffuse, parameterName=Complex Diffuse Multiplier, sortPriority=6, defaultValue=<rgba 1 1 1 1>}
  n69 = ScalarParameter() {group=Base_Diffuse, parameterName=Use Complex Diffuse Multipier, sortPriority=5}
  n70 = MaterialFunctionCall(
   Hue Shift Percentage=ScalarParameter() {group=Base_Diffuse, parameterName=Hue Shift Percentage, sortPriority=6}
   Texture=Multiply(A=LinearInterpolate(A=@n55, B=@n65, Alpha=@n66), B=LinearInterpolate(A=@n67, B=@n68.RGB, Alpha=@n69))
  ) {materialFunction=&HueShift}
  NamedRerouteDeclaration(None=LinearInterpolate(
    A=@n54.RGB
    B=Multiply(A=@n70.Result, B=@n54.RGB)
    Alpha=ScalarParameter() {group=Emissive, parameterName=UseDiffuseinEmissive, sortPriority=1}
   )) {name=Diffuse to Emissive, nodeColor=<rgba 1 0 0.7031 1>}
  n71 = Multiply(A=ParticleColor().RGB, B=@n70.Result)
  n72 = NamedRerouteUsage()
  NamedRerouteDeclaration(None=LinearInterpolate(A=@n71, B=Multiply(A=@n71, B=@n72), Alpha=ScalarParameter() {group=Emissive Sphere Mask, parameterName=Base Color Influence})) {name=Base Color, nodeColor=<rgba 0 1 0.9766 1>}
  n73 = CameraVectorWS()
  n74 = AtmosphericLightVector()
  n75 = ScalarParameter() {group=Emissive, parameterName=Emissive Mask Fresnel Fake Slerp, sortPriority=11}
  n76 = LinearInterpolate(
   B=MaterialFunctionCall(
    Camera Vector=LinearInterpolate(A=@n73, B=@n74, Alpha=@n75)
    Power=ScalarParameter() {defaultValue=1, group=Emissive, parameterName=Emissive Mask Fresnel Exponent, sortPriority=12}
    Cheap contrast dark=ScalarParameter() {defaultValue=0.04, group=Emissive, parameterName=Emissive Mask Fresnel Min Value, sortPriority=13}
    Cheap contrast bright=ScalarParameter() {defaultValue=1, group=Emissive, parameterName=Emissive Mask Fresnel Max Value, sortPriority=14}
   ) {materialFunction=&Fresnel_Function}.Result
   Alpha=ScalarParameter() {group=Emissive, parameterName=Emissive Mask Fresnel, sortPriority=10}
  ) {constA=1}
  n77 = ScalarParameter() {
   desc=This option is there if you are using Emissive Mask Fresnel and you like to avoid using the constant value of 1 for lower end platform in replacement to the fresnel.
   group=Emissive
   parameterName=Disable Emissive Mask Fresnel Constant on Low End Platforms
   sortPriority=15
  }
  n78 = Constant()
  n79 = If(A=@n77, "A > B"=@n78, "A == B"=@n78, "A < B"=Constant() {r=1}) {constB=0.99}
  NamedRerouteDeclaration(None=ShadingPathSwitch(Default=QualitySwitch(Default=@n76, Low=@n79), Mobile=@n79) {desc=Migrated from FeatureLevelSwitch}) {name=Emissive Fresnel Function, nodeColor=<rgba 0 1 0.5078 1>}
  n80 = NamedRerouteUsage()
  n81 = VectorParameter() {
   group=Base_Specular
   parameterName=Specular Metallic Roughness Mask Tile
   sortPriority=4
   channelNames {a=SubUV Spec Y, b=SubUV Spec X, g=Spec Y, r=Spec X}
   defaultValue=<rgba 1 1 1 1>
  }
  n82 = AppendVector(A=@n81.Spec X, B=@n81.Spec X)
  n83 = ScalarParameter() {group=Base_UVs, parameterName=U Speed, sortPriority=1}
  n84 = ScalarParameter() {group=Base_UVs, parameterName=V Speed, sortPriority=2}
  n85 = TextureSampleParameter2D(UVs=Panner(Coordinate=Multiply(A=@n80, B=@n82), Speed=AppendVector(A=@n83, B=@n84))) {
   desc="R=Specular, G=Metallic, B=Roughness"
   group=Base_Specular
   parameterName=Specular Metallic Roughness Mask
   samplerSource=SSM_Wrap_WorldGroupSettings
   samplerType=SAMPLERTYPE_Masks
   sortPriority=1
   texture=&T_Fortnite_Default_S
  }
  n86 = TextureSampleParameterSubUV(UVs=Multiply(A=@n80, B=AppendVector(A=@n81.SubUV Spec X, B=@n81.SubUV Spec X))) {
   group=Base_Specular
   parameterName=Specular Metallic Roughness Mask_SubUV
   samplerSource=SSM_Wrap_WorldGroupSettings
   samplerType=SAMPLERTYPE_Masks
   sortPriority=3
   texture=&T_Fortnite_Default_S
  }
  n87 = MaterialFunctionCall(Color=LinearInterpolate(
    A=@n85.RGB
    B=@n86.RGB
    Alpha=ScalarParameter() {group=Base_Specular, parameterName=Use SubUV for Specular Metallic Roughness Mask, sortPriority=2}
   ) {constAlpha=0}) {materialFunction=&SplitComponents}
  n88 = LinearInterpolate(
   A=ScalarParameter() {group=Base_Specular, parameterName=SpecRoughnessMin, sortPriority=5}
   B=ScalarParameter() {defaultValue=1, group=Base_Specular, parameterName=SpecRoughnessMax, sortPriority=6}
   Alpha=@n87.B
  )
  n89 = ScalarParameter() {
   desc="Assume this needs to be 1 TRUE, unless working with older Specular Maps"
   group=Base_Specular
   parameterName="Swap Metallic(G->B) and Roughness (B->G)"
   sortPriority=0
  }
  NamedRerouteDeclaration(None=LinearInterpolate(A=@n88, B=@n87.G, Alpha=@n89)) {name="Metallic (G)", nodeColor=<rgba 1 0.1875 0 1>}
  NamedRerouteDeclaration(None=@n87.R) {name="Specular (R)", nodeColor=<rgba 0 0.1328 1 1>}
  NamedRerouteDeclaration(None=LinearInterpolate(A=@n88, B=@n87.G, Alpha=@n89)) {name="Roughness (B)", nodeColor=<rgba 0.4531 1 0 1>}
  n90 = NamedRerouteUsage()
  n91 = TextureSampleParameter2D(UVs=@n90) {
   group=Emissive
   parameterName=Emissive Mask Texture
   samplerSource=SSM_Wrap_WorldGroupSettings
   sortPriority=3
   texture=&T_WhitesRGB
  }
  n92 = TextureSampleParameterSubUV(UVs=@n90) {
   group=Emissive
   parameterName=Emissive Mask SubUV
   samplerSource=SSM_Wrap_WorldGroupSettings
   sortPriority=6
   texture=&T_WhitesRGB
  }
  n93 = Multiply(
   A=NamedRerouteUsage()
   B=LinearInterpolate(A=@n91.RGBA, B=@n92.RGBA, Alpha=ScalarParameter() {group=Emissive, parameterName=Use SubUV Emissive Mask Texture, sortPriority=5}) {constAlpha=0}
  ) {constA=1}
  n94 = LinearInterpolate(
   A=ChannelMaskParameter(None=@n93) {group=Emissive, parameterName=Emissive Mask Texture Channel, sortPriority=4}.R
   B=ComponentMask(None=@n93) {b=true, g=true, r=true}
   Alpha=ScalarParameter() {group=Emissive, parameterName=Emissive Mask Uses RGB, sortPriority=7}
  )
  n95 = Multiply(
   A=Multiply(
    A=NamedRerouteUsage()
    B=DynamicParameter() {defaultValue=<rgba 1 1 0 1>, paramNames [Emissive Intensity, ColorChartUV, Erode, WPOScale]}.Emissive Intensity
   )
   B=LinearInterpolate(A=@n94, B=OneMinus(None=@n94), Alpha=ScalarParameter() {group=Emissive, parameterName=Invert Emissive Mask, sortPriority=8})
  )
  n96 = ParticleColor()
  n97 = LinearInterpolate(
   A=Multiply(A=@n95, B=@n96.RGB)
   B=Add(A=@n95, B=@n96.RGB)
   Alpha=ScalarParameter() {group=Emissive, parameterName=Particle Emissive Mult Or Add, sortPriority=9}
  )
  n98 = NamedRerouteUsage()
  NamedRerouteDeclaration(None=Multiply(A=Multiply(A=@n97, B=@n98), B=NamedRerouteUsage())) {name=Emissive, nodeColor=<rgba 0 0.1683 0.0931 1>}
  n99 = MaterialFunctionCall(Float3=Subtract(A=WorldPosition().XYZ, B=ObjectPositionWS())) {materialFunction=&BreakOutFloat3Components}
  n100 = Divide(A=@n99.B, B=DynamicParameter() {parameterIndex=1, paramNames [Amplitude,Frequency,Offset,Axis]}.Axis) {constB=100, desc=Falloff}
  n101 = ParticleRandom()
  n102 = DynamicParameter() {parameterIndex=1, paramNames [Amplitude,Frequency,Offset,Axis]}
  n103 = Divide(A=@n99.B, B=@n102.Frequency) {constB=50}
  n104 = ScalarParameter() {group=WorldPositionOffset_Advanced, parameterName=Noise Offset, sortPriority=4}
  n105 = TextureObjectParameter() {
   group=WorldPositionOffset_Advanced
   parameterName=Vertex Offset
   samplerType=SAMPLERTYPE_Normal
   sortPriority=3
   texture=/Game/Effects/Fort_Effects/Textures/T_DustDistortion
  }
  n106 = TextureSample(UVs=Add(A=AppendVector(A=@n101, B=@n103), B=Add(A=@n102.Offset, B=@n104)), Tex=@n105)
  n107 = DynamicParameter() {desc=Axis of influence, parameterIndex=2, paramNames [Amplitude,Frequency,Offset,Axis]}
  n108 = MaterialFunctionCall(X=@n107.Amplitude, Y=@n107.Amplitude, Z=@n107.Amplitude) {materialFunction=&MakeFloat3}
  n109 = Multiply(
   A=Multiply(A=Clamp(None=@n100), B=Multiply(A=@n106.RGB, B=@n108.Result)) {constB=3}
   B=DynamicParameter() {parameterIndex=1, paramNames [Amplitude,Frequency,Offset,Axis]}.Amplitude
  )
  n110 = LinearInterpolate(
   B=LinearInterpolate(
    A=@n109
    B=Transform(None=@n109) {transformSourceType=TRANSFORMSOURCE_Local}
    Alpha=ScalarParameter() {group=WorldPositionOffset_Advanced, parameterName=Is In Local Space, sortPriority=2}
   )
   Alpha=ScalarParameter() {group=WorldPositionOffset_Advanced, parameterName=Use Advanced WPO, sortPriority=0}
  )
  n111 = MaterialFunctionCall!(InVector=@n110) {materialFunction=/Game/Athena/Environments/Landscape/Foliage/MF_MakePreciseVec3}
  n112 = Constant() {r=1}
  n113 = VertexColor()
  n114 = ScalarParameter() {desc=Enables Vertex COlor R to mask WPO., group=WorldPositionOffset_Advanced, parameterName=Vertex Color R Masks WPO, sortPriority=5}
  NamedRerouteDeclaration(None=Multiply(A=ShadingPathSwitch(Default=@n111.Result, Mobile=@n110), B=LinearInterpolate(A=@n112, B=@n113.R, Alpha=@n114))) {name=Advanced WPO, nodeColor=<rgba 1 0.7031 0 1>}
 ]
}
