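; Anim notify state that pushes configured parameter values onto the Niagara effect it spawns during an animation.
; On notify tick, a DoOnce gate with no reset lets it loop through the Parameter Setup array only the first time.
; Each entry's type enum selects setting a named float, vector, linear color or integer parameter on the spawned effect fetched from the mesh component.
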(fn Received_NotifyTick
  (params (MeshComp :type Skeletal%20Mesh%20Component%20Object%20Reference) (Animation :type Anim%20Sequence%20Base%20Object%20Reference) (FrameDeltaTime :type Float%20%28single-precision%29) (EventReference :type Anim%20Notify%20Event%20Reference%20Structure%20%28by%20ref%29))
  (Utilities|FlowControl|DoOnce false
    (:Completed
      (for _array_element (Variables|AnimNotify|GetParameter%20Setup)
        (switch Utilities|FlowControl|Switch|SwitchonENUM_NiagaraParameterSetup (Utilities|Struct|BreakStrucNiagaraParameters _array_element)
          (:NewEnumerator0
            (bind _returnvalue (AnimNotify|GetSpawnedEffect MeshComp))
            (Effects|Components|ParticleSystem|SetFloatParameter _returnvalue (.FloatParameterName_17_6AFBA48F43B44BD1ED9841983BEFA202 (Utilities|Struct|BreakStrucNiagaraParameters _array_element)) (.FloatParameterValue_28_AC98B6464E1CD803B50B47AEC91741EC (Utilities|Struct|BreakStrucNiagaraParameters _array_element))))
          (:NewEnumerator1
            (bind _returnvalue_2 (AnimNotify|GetSpawnedEffect MeshComp))
            (Effects|Components|ParticleSystem|SetVectorParameter _returnvalue_2 (.VectorParameterName_20_71917CFB490B3D5D3E0864A3EE90C666 (Utilities|Struct|BreakStrucNiagaraParameters _array_element)) (.VectorParameterValue_27_248856604659F589FF5D4FA5FF52E013 (Utilities|Struct|BreakStrucNiagaraParameters _array_element))))
          (:NewEnumerator2
            (bind _returnvalue_3 (AnimNotify|GetSpawnedEffect MeshComp))
            (Effects|Components|ParticleSystem|SetColorParameter _returnvalue_3 (.LinearColorparameterName_22_69C61F8D45C74F98203920856CAD5ADE (Utilities|Struct|BreakStrucNiagaraParameters _array_element)) (.LinearColorParameterValue_29_861B08ED408F0F612BF509A2A94F377F (Utilities|Struct|BreakStrucNiagaraParameters _array_element))))
          (:NewEnumerator3
            (bind _returnvalue_4 (AnimNotify|GetSpawnedEffect MeshComp))
            (Effects|Components|ParticleSystem|SetIntParameter _returnvalue_4 (.IntegerParameterName_24_98306BA7452D90289F6EE9954C4F2B1D (Utilities|Struct|BreakStrucNiagaraParameters _array_element)) (.IntegerParameterValue_25_BB84ADEA4515AABD3B9AC6A04453D826 (Utilities|Struct|BreakStrucNiagaraParameters _array_element))))))
      (return false))))
