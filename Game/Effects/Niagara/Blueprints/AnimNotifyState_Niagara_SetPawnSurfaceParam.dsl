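; Niagara anim notify state that feeds the owning Fort pawn's surface and water context to its spawned effect.
; On notify begin, it refreshes surface tracking so clients get the current type, then sets on-ground and physical-material-index parameters.
; In a water body, it either deactivates the effect via an option or passes water body index, surface location, normal and depth.
; Values are not updated on tick.
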(fn Received_NotifyBegin
  (params (MeshComp :type Skeletal%20Mesh%20Component%20Object%20Reference) (Animation :type Anim%20Sequence%20Base%20Object%20Reference) (TotalDuration :type Float%20%28single-precision%29) (EventReference :type Anim%20Notify%20Event%20Reference%20Structure%20%28by%20ref%29))
  (bind _asfort_pawn (Utilities|Casting|CastToFortPawn (Components|GetOwner MeshComp))
    (:then
      (bind _returnvalue (AnimNotify|GetSpawnedEffect MeshComp))
      (Utilities|IsValid _returnvalue
        (:Is%20Valid
          (Class|FortSurfaceTrackingComponent|UpdateSurfaceType (Footsteps|GetSurfaceTrackingComponent _asfort_pawn) false false) ; Needed to make sure surface type gets update on clients
          (Effects|Components|ParticleSystem|SetBoolParameter _returnvalue "bIsOnGround" (AI|Components|NavMovement|IsMovingonGround (Pawn|GetMovementComponent _asfort_pawn)))
          (Effects|Components|ParticleSystem|SetIntParameter _returnvalue "PhysMatIndex" (Math|Conversions|ToInteger(Byte) (Class|FortSurfaceTrackingComponent|GetCurrentSurfaceType (Footsteps|GetSurfaceTrackingComponent _asfort_pawn))))
          (if (Utilities|IsValid (WaterBody|GetCurrentWaterBody _asfort_pawn) :!class /Script/Engine.KismetSystemLibrary)
            (if (Variables|Default|GetDisable%20if%20in%20Water)
              (Components|Activation|SetActive _returnvalue (not (Utilities|IsValid (WaterBody|GetCurrentWaterBody _asfort_pawn) :!class /Script/Engine.KismetSystemLibrary)) false)
              (else
                (Effects|Components|ParticleSystem|SetBoolParameter _returnvalue "bIsInWater" true)
                (Effects|Components|ParticleSystem|SetIntParameter _returnvalue "WaterBodyIndex" (Variables|/Script/Water.WaterBody|GetWaterBodyIndex (WaterBody|GetCurrentWaterBody _asfort_pawn)))
                (Effects|Components|ParticleSystem|SetVectorParameter _returnvalue "WaterPlaneLocation" (WaterBody|GetWaterSurfaceInfoatLocation (Variables|/Script/Water.WaterBody|GetWaterBodyComponent (WaterBody|GetCurrentWaterBody _asfort_pawn)) (Transformation|GetWorldLocation _returnvalue) true))
                (Effects|Components|ParticleSystem|SetVectorParameter _returnvalue "WaterPlaneNormal" (.OutWaterSurfaceNormal (WaterBody|GetWaterSurfaceInfoatLocation (Variables|/Script/Water.WaterBody|GetWaterBodyComponent (WaterBody|GetCurrentWaterBody _asfort_pawn)) (Transformation|GetWorldLocation _returnvalue) true)))
                (Effects|Components|ParticleSystem|SetFloatParameter _returnvalue "WaterDepth" (.OutWaterDepth (WaterBody|GetWaterSurfaceInfoatLocation (Variables|/Script/Water.WaterBody|GetWaterBodyComponent (WaterBody|GetCurrentWaterBody _asfort_pawn)) (Transformation|GetWorldLocation _returnvalue) true)))))))))
    (:CastFailed
      (return false))))

(fn Received_NotifyTick
  (params (MeshComp :type Skeletal%20Mesh%20Component%20Object%20Reference) (Animation :type Anim%20Sequence%20Base%20Object%20Reference) (FrameDeltaTime :type Float%20%28single-precision%29) (EventReference :type Anim%20Notify%20Event%20Reference%20Structure%20%28by%20ref%29))
  (return false))
