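# Physics asset for the St. Patrick variant of the female medium soldier hat (BR), providing its collision shape.
# It holds a single body on the root bone, made of one tilted capsule offset from the bone origin.
# It defines no constraints, so the asset contains no jointed or secondary simulated parts.
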PhysicsAsset /Game/F_MED_Soldier_Hat_BR_ST_Patrick_PhysicsAsset
bodies {root [{shapeType=Capsule, length=7.7013, radius=7.116, center=<-3.125 -1.7871 19.8069>, rotation=<rot 0 120.3394 -54.3773>}]}
constraints []
