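# Material function that outputs a spherical world-space glitch mask, apparently for post-process materials given its PPM name.
# It measures the distance from world position to the GlitchOrigin collection parameter and divides it by the GlitchRadius input scaled by the GlitchStrength collection parameter.
# The inverted, saturated falloff is raised to the GlitchEdgeHardness scalar parameter, peaking at the origin and reaching zero at the scaled radius.
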MaterialFunction /Game/MF_PPM_GlitchMask
graph {nodes [
  n1 = Subtract(A=WorldPosition().XYZ, B=ComponentMask(None=CollectionParameter() {parameterName=GlitchOrigin}) {b=true, g=true, r=true})
  n2 = FunctionInput() {inputName=GlitchRadius, inputType=FunctionInput_Scalar}
  n3 = CollectionParameter() {parameterName=GlitchStrength}
  n4 = OneMinus(None=Divide(A=Length(None=@n1) {is=class MaterialExpressionLength}, B=Multiply(A=@n2, B=@n3)))
  FunctionOutput(None=Power(Base=Saturate(None=@n4), Exp=ScalarParameter() {parameterName=GlitchEdgeHardness})) {outputName=GlitchMask}
 ]}
