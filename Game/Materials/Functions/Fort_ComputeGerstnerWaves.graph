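# Water material function that computes Gerstner wave normals, world position offset and a camera distance mask.
# Custom HLSL splits waves into three frequency bands; normals sum low plus doubled middle, while offset sums doubled low plus high.
# Static switches move normals to vertex interpolators; quality and shading-path switches simplify to flat normals and low-frequency offset.
# Horizontal offset fades out near the camera, with an optional Z-only output.
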MaterialFunction /Game/Materials/Functions/Fort_ComputeGerstnerWaves
graph {
 properties {bExposeToLibrary=true}
 nodes [
  MaterialExpressionFunctionInput_1 = FunctionInput(Preview=StaticSwitchParameter(
    True=ScalarParameter() {parameterName=WaterBodyIndex}
    False=Custom() {code="GetWaterWaveParamIndex(Parameters)", description=WaterBodyIndex From Tile Instance, outputType=CMOT_Float1, inputs []}
   ) {parameterName=UseScalarWaterBodyIndex}) {bUsePreviewValueAsDefault=true, inputName=WaterBodyIndex, inputType=FunctionInput_Scalar, sortPriority=5}
  MaterialExpressionFunctionInput_0 = FunctionInput(Preview=WorldPosition() {worldPositionShaderOffset=WPT_ExcludeAllShaderOffsets}.XYZ) {bUsePreviewValueAsDefault=true, inputName=Position}
  MaterialExpressionMaterialFunctionCall_6 = MaterialFunctionCall() {materialFunction=/Water/Materials/Functions/GetWaterTime}
  MaterialExpressionConstant_14 = Constant() {r=3}
  MaterialExpressionConstant_13 = Constant() {r=6}
  n1 = Custom!(
   WaterBodyIndex=@MaterialExpressionFunctionInput_1
   WorldPosition=@MaterialExpressionFunctionInput_0
   Time=@MaterialExpressionMaterialFunctionCall_6.WaterTime
   LowFrequencyMaxIndex=@MaterialExpressionConstant_14
   MiddleFrequencyMaxIndex=@MaterialExpressionConstant_13
  ) {
   code="WaveOutputs Waves;\r\n\r\nWaves = GetPerFrequencyGerstnerWavesNew((int)LowFrequencyMaxIndex, (int)MiddleFrequencyMaxIndex, WaterBodyIndex, Time, MakeLWCVector2(LWCWorldPosition.Tile, LWCWorldPosition.Offset));\r\n\r\nLowFrequencyWPO = Waves.LowFrequency.WPO;\r\nMiddleFrequencyWPO = Waves.MiddleFrequency.WPO;\r\nHighFrequencyWPO = Waves.HighFrequency.WPO;\r\n\r\nLowFrequencyNormal = Waves.LowFrequency.Normal;\r\nMiddleFrequencyNormal = Waves.MiddleFrequency.Normal;\r\nHighFrequencyNormal = Waves.HighFrequency.Normal;\r\nreturn 0;"
   description=Get Per Frequency Gerstner Waves
   outputType=CMOT_Float1
   additionalOutputs [
    {outputName=LowFrequencyNormal, outputType=CMOT_Float3}
    {outputName=MiddleFrequencyNormal, outputType=CMOT_Float3}
    {outputName=HighFrequencyNormal, outputType=CMOT_Float3}
    {outputName=LowFrequencyWPO, outputType=CMOT_Float3}
    {outputName=MiddleFrequencyWPO, outputType=CMOT_Float3}
    {outputName=HighFrequencyWPO, outputType=CMOT_Float3}
   ]
   includeFilePaths [/Plugin/Water/Private/GerstnerWaveFunctions.ush]
   inputs [
    {inputName=WaterBodyIndex, input {expression=&self:MaterialExpressionFunctionInput_1}}
    {inputName=WorldPosition, input {expression=&self:MaterialExpressionFunctionInput_0}}
    {inputName=Time, input {expression=&self:MaterialExpressionMaterialFunctionCall_6}}
    {inputName=LowFrequencyMaxIndex, input {expression=&self:MaterialExpressionConstant_14}}
    {inputName=MiddleFrequencyMaxIndex, input {expression=&self:MaterialExpressionConstant_13}}
   ]
  }
  n2 = Add(A=@n1.MiddleFrequencyNormal, B=@n1.MiddleFrequencyNormal)
  n3 = Add(A=@n2, B=@n1.LowFrequencyNormal)
  n4 = VertexInterpolator(VS=@n3)
  n5 = Add(A=VertexInterpolator(VS=@n1.LowFrequencyNormal).PS, B=@n2)
  n6 = FunctionInput(Preview=StaticBool()) {
   bUsePreviewValueAsDefault=true
   description="If Normals use VS Interp is checked,  this allows forcing all 6 waves to use the interpolation."
   inputName=All Normals VS
   inputType=FunctionInput_StaticBool
   sortPriority=12
  }
  n7 = FunctionInput(Preview=StaticBool()) {
   bUsePreviewValueAsDefault=true
   description="If this is true, the first 4 waves will have their normal computed on the vertex shader."
   inputName=Normals use VS Interpolation
   inputType=FunctionInput_StaticBool
   sortPriority=11
  }
  n8 = Constant3Vector() {constant=<rgba 0 0 1 0>}
  n9 = QualitySwitch(Default=StaticSwitch(True=StaticSwitch(True=@n4.PS, False=@n5, Value=@n6), False=@n3, Value=@n7), Low=@n8)
  FunctionOutput(None=ShadingPathSwitch(Default=@n9, Deferred=@n9, Forward=@n8, Mobile=@n8)) {outputName=Normal}
  n10 = QualitySwitch(Default=Add(A=Add(A=@n1.LowFrequencyWPO, B=@n1.LowFrequencyWPO), B=@n1.HighFrequencyWPO), Low=@n1.LowFrequencyWPO)
  n11 = Constant() {r=11}
  n12 = Constant() {r=64}
  n13 = Multiply(
   A=ShadingPathSwitch(Default=@n10, Deferred=@n10, Forward=@n10, Mobile=@n1.LowFrequencyWPO)
   B=LinearInterpolate(
    A=Constant3Vector() {constant=<rgba 0 0 1 0>}
    Alpha=MaterialFunctionCall(Fade Length=@n11, Fade Offset=@n12) {materialFunction=&CameraDepthFade}.Result
   )
  )
  n14 = StaticBool()
  FunctionOutput(None=StaticSwitch(
    True=ComponentMask(None=@n13) {b=true}
    False=@n13
    Value=FunctionInput(Preview=@n14) {bUsePreviewValueAsDefault=true, inputName=Z Offset Only, inputType=FunctionInput_StaticBool}
   )) {outputName=WPO, sortPriority=1}
  FunctionOutput(None=MaterialFunctionCall(
    Fade Length=ScalarParameter() {defaultValue=15000, group=Distance Blending, parameterName=Far Fade Length}
    Fade Offset=ScalarParameter() {defaultValue=7500, group=Distance Blending, parameterName=Far Fade Start}
   ) {materialFunction=&CameraDepthFade}.Result) {outputName=Distance Mask, sortPriority=2}
 ]
}
