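# Material function that produces the final water displacement, combining Gerstner waves, flowmap displacement and water height into WPO, normal, foam, depth, height and velocity outputs.
# A velocity-and-height texture sampled at world position drives the flowmap data and depth masks.
# Depth-attenuated Gerstner waves blend toward flowmap results by depth mask and an overridable, velocity-scaled camera distance fade.
# Foam combines phase offset, velocity above a threshold and flowmap vertical displacement.
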MaterialFunction /Game/Materials/Functions/FinalWaterDisplacement
graph {nodes [
  n1 = Divide(
   A=ComponentMask(None=@n8.VelocityAndLength) {b=true}
   B=ScalarParameter() {defaultValue=1, group=Distance Blending, parameterName=Velocity Distance Divisor}
  )
  n2 = Add(A=Multiply(A=Saturate(None=@n1), B=ScalarParameter() {defaultValue=2, group=Distance Blending, parameterName=Flowmap Fade Push}))
  n3 = ScalarParameter() {defaultValue=15000, group=Distance Blending, parameterName=Far Fade Length}
  n4 = ScalarParameter() {defaultValue=10000, group=Distance Blending, parameterName=Far Fade Start}
  n5 = FunctionInput(Preview=MaterialFunctionCall(Fade Length=Multiply(A=@n2, B=@n3), Fade Offset=Multiply(A=@n2, B=@n4)) {materialFunction=&CameraDepthFade}.Result) {
   bUsePreviewValueAsDefault=true
   description="When this is left disconnected, the function blends to a simplified texture in the distance. Plug in a constant 0 to disable."
   inputName=Override Distance Fade
   inputType=FunctionInput_Scalar
   sortPriority=2
  }
  n6 = FunctionInput(Preview=TextureObjectParameter() {group=Technical, parameterName=WaterVelocityAndHeight, texture=/Game/Textures/RT_DefaultWaterDepth}) {
   bUsePreviewValueAsDefault=true
   description=This input lets you override the texture containing the combined water height and velocity information.
   inputName=Texture
   inputType=FunctionInput_Texture2D
  }
  n7 = FunctionInput(Preview=WorldPosition() {worldPositionShaderOffset=WPT_ExcludeAllShaderOffsets}.XYZ) {
   bUsePreviewValueAsDefault=true
   description="By default, worldposition is used. Overriding this input can allow you to sample from custom locations."
   inputName=Override Position
   sortPriority=1
  }
  n8 = MaterialFunctionCall!(Far Mask=Ceil(None=@n5), Texture=@n6, Position=@n7)
  ComponentMask(None=@n8.Phase Offset And Foam) {g=true}
  n9 = ComponentMask(None=@n8.Flowmap WPO) {b=true}
  Divide(A=@n9, B=ComponentMask(None=@n8.Exponential Depth Masks) {g=true})
  n10 = Normalize(VectorInput=@n8.Flowmap Normals)
  n11 = MaterialFunctionCall!(Z Offset Only, Position=@n7, WaterBodyIndex, Normals use VS Interpolation, All Normals VS) {materialFunction=&/Fort_ComputeGerstnerWaves}
  n12 = Multiply(A=@n10, B=VectorParameter() {group=Normals, parameterName=DetailNormalFrom Flowmaps, defaultValue=<rgba 0.5 0.5 0 0>}.RGB)
  n13 = ComponentMask(None=@n8.Exponential Depth Masks) {r=true}
  n14 = MaterialFunctionCall(Normal=Normalize(VectorInput=Add(A=@n11.Normal, B=@n12)), Flatness=OneMinus(None=@n13)) {materialFunction=&FlattenNormal}
  n15 = Min(
   A=OneMinus(None=ComponentMask(None=@n8.Exponential Depth Masks) {a=true})
   B=OneMinus(None=FunctionOutput(None=@n5) {outputName=Distance Mask, sortPriority=10})
  )
  FunctionOutput(None=Multiply(A=LinearInterpolate(A=@n10, B=@n14.Result, Alpha=@n15)) {desc=N}) {outputName=Normal}
  n16 = Multiply(A=@n11.WPO, B=@n13)
  FunctionOutput(None=Add(A=LinearInterpolate(A=@n8.Flowmap WPO, B=@n16, Alpha=@n15), B=@n8.Water Height WPO) {desc=WPO}) {outputName=WPO, sortPriority=1}
  FunctionOutput(None=@n8.Water Depth) {outputName=Water Depth, sortPriority=2}
  FunctionOutput(None=@n8.Water Z Height) {outputName=Water Z Height, sortPriority=3}
  FunctionOutput(None=@n8.VelocityAndLength) {outputName=Water Velocity, sortPriority=4}
  n17 = Subtract(A=ComponentMask(None=@n8.VelocityAndLength) {b=true}, B=ScalarParameter() {defaultValue=256, parameterName=MinFoamVelocity})
  n18 = ScalarParameter() {defaultValue=0.01, parameterName=Velocity Foam}
  n19 = ScalarParameter() {defaultValue=0.1, parameterName=Foam WPO Amount}
  n20 = Add(
   A=Multiply(A=ComponentMask(None=@n8.Phase Offset And Foam) {r=true}, B=ScalarParameter() {defaultValue=-3, parameterName=Phase offset Foam Amount})
   B=Add(A=Multiply(A=@n17, B=@n18), B=Multiply(A=@n9, B=@n19))
  )
  n21 = AppendVector(A=Saturate(None=@n20), B=@n20)
  FunctionOutput(None=@n21) {description="Displacement from Waves and Flowmaps, but not from Water Height", outputName=Foam, sortPriority=10}
  Saturate(None=@n21)
 ]}
