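# Material function that fetches one Gerstner wave's parameters for a given water body and wave index.
# A Custom node calls GetWaveDataByWaveIndexNew from the Water plugin's GerstnerWaveFunctions include.
# Water body index defaults to the tile instance's value through GetWaterWaveParamIndex, and wave index defaults to a constant.
# It outputs direction, wavelength, amplitude and steepness, plus the latter three packed through AppendMany as Params.
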MaterialFunction /Game/Materials/Functions/GetWaveData
graph {nodes [
  MaterialExpressionFunctionInput_1 = FunctionInput(Preview=Custom() {code="GetWaterWaveParamIndex(Parameters)", description=WaterBodyIndex From Tile Instance, outputType=CMOT_Float1, inputs []}) {bUsePreviewValueAsDefault=true, inputName=Water Body Index, inputType=FunctionInput_Scalar, sortPriority=1}
  MaterialExpressionFunctionInput_0 = FunctionInput(Preview=Constant()) {bUsePreviewValueAsDefault=true, inputName=Wave Index, inputType=FunctionInput_Scalar}
  n1 = Custom!(WaterBodyIndex=@MaterialExpressionFunctionInput_1, WaveIndex=@MaterialExpressionFunctionInput_0) {
   code="FWaveParams WaveParams = GetWaveDataByWaveIndexNew(WaterBodyIndex, WaveIndex);\r\n\r\nDirection = WaveParams.Direction;\r\nWavelength = WaveParams.Wavelength;\r\nAmplitude = WaveParams.Amplitude;\r\nSteepness = WaveParams.Steepness;\r\n\r\nreturn WaveParams.Direction;"
   description=Get Gerstner Wave Params
   outputType=CMOT_Float2
   additionalOutputs [
    {outputName=Direction, outputType=CMOT_Float2}
    {outputName=Wavelength, outputType=CMOT_Float1}
    {outputName=Amplitude, outputType=CMOT_Float1}
    {outputName=Steepness, outputType=CMOT_Float1}
   ]
   includeFilePaths [/Plugin/Water/Private/GerstnerWaveFunctions.ush]
   inputs [
    {inputName=WaterBodyIndex, input {expression=&self:MaterialExpressionFunctionInput_1}}
    {inputName=WaveIndex, input {expression=&self:MaterialExpressionFunctionInput_0}}
   ]
  }
  FunctionOutput(None=@n1.Direction) {outputName=Direction}
  FunctionOutput(None=MaterialFunctionCall(
    R=FunctionOutput(None=@n1.Wavelength) {outputName=Wavlength, sortPriority=2}
    G=FunctionOutput(None=@n1.Amplitude) {outputName=Amplitude, sortPriority=3}
    B=FunctionOutput(None=@n1.Steepness) {outputName=Steepness, sortPriority=4}
   ) {materialFunction=&AppendMany}.RGBA) {outputName=Params, sortPriority=1}
 ]}
