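# Material function that sums six Gerstner waves into world position offset and a normal for water surfaces, plus a distance mask.
# Wave data comes from GetWaveData by water body and offset index, or from MPC_Water, via static switch.
# Packed float4 custom nodes batch the sine and cosine math.
# Camera depth fades the last two waves' normals; normals can use vertex interpolation, and flatten on low quality, forward and mobile.
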MaterialFunction /Game/Materials/Functions/MF_Gerstner_6_Waves_Multi
graph {nodes [
  n1 = FunctionInput(Preview=StaticBool()) {bUsePreviewValueAsDefault=true, inputName=Z Offset Only, inputType=FunctionInput_StaticBool}
  n2 = FunctionInput() {bUsePreviewValueAsDefault=true, inputName=Index Offset, inputType=FunctionInput_Scalar, sortPriority=7}
  n3 = FunctionInput(Preview=ScalarParameter() {group=Waves, parameterName=WaterBodyIndex}) {bUsePreviewValueAsDefault=true, inputName=WaterBodyIndex, inputType=FunctionInput_Scalar, sortPriority=5}
  n4 = MaterialFunctionCall!(Wave Index=Add(A=Constant() {r=4}, B=@n2), Water Body Index=@n3) {materialFunction=&/GetWaveData}
  n5 = CollectionParameter() {collection=/Water/Materials/MPC/MPC_Water, parameterName=k4}
  n6 = StaticBoolParameter() {defaultValue=true, parameterName=UseTextureWaveData}
  n7 = MaterialFunctionCall(Float4=StaticSwitch(True=@n4.Params, False=CollectionParameter() {collection=&MPC_Water, parameterName=w4}, Value=@n6)) {materialFunction=&BreakOutFloat4Components}
  n8 = FunctionInput(Preview=WorldPosition() {worldPositionShaderOffset=WPT_ExcludeAllShaderOffsets}.XYZ) {bUsePreviewValueAsDefault=true, inputName=Position}
  n9 = MaterialFunctionCall(Wave Index=Add(A=Constant() {r=5}, B=@n2), Water Body Index=@n3) {materialFunction=&GetWaveData}
  n10 = CollectionParameter() {collection=&MPC_Water, parameterName=k5}
  n11 = MaterialFunctionCall(Float4=StaticSwitch(True=@n9.Params, False=CollectionParameter() {collection=&MPC_Water, parameterName=w5}, Value=@n6)) {materialFunction=&BreakOutFloat4Components}
  n12 = Custom!(x=@MaterialExpressionMaterialFunctionCall_34.Result) {
   code="cos(x);"
   description="cos(x)"
   outputType=CMOT_Float4
   inputs [{inputName=x, input {expression=&self:MaterialExpressionMaterialFunctionCall_34}}]
  }
  n13 = MaterialFunctionCall(Float4=@n12) {materialFunction=&BreakOutFloat4Components}
  n14 = MaterialFunctionCall!(
   Z Only Offset=@n1
   2D Direction Vector=StaticSwitch(True=@n9.Direction, False=ComponentMask(None=@n10) {g=true, r=true}, Value=@n6)
   Wave Length=@n11.R
   Amplitude=@n11.R
   Wave Steepness=@n11.R
   Pos=@n8
   wavesin 1=@n16.G
   wavesin 2=@n16.G
   wavecos 1=@n13.G
   wavecos 2=@n13.G
   Manual Time
  ) {desc=5, materialFunction=&/GerstnerWave_Deep_WithZTrace}
  MaterialExpressionMaterialFunctionCall_34 = MaterialFunctionCall(X=@n17.Wave Pos 1, Y=@n14.Wave Pos 1, Z=@n17.Wave Pos 1, A=@n14.Wave Pos 1) {materialFunction=&MakeFloat4}
  n15 = Custom(x=@MaterialExpressionMaterialFunctionCall_34.Result) {
   code="sin(x);"
   description="Sin(x)"
   outputType=CMOT_Float4
   inputs [{inputName=x, input {expression=&self:MaterialExpressionMaterialFunctionCall_34}}]
  }
  n16 = MaterialFunctionCall(Float4=@n15) {materialFunction=&BreakOutFloat4Components}
  n17 = MaterialFunctionCall(
   Z Only Offset=@n1
   2D Direction Vector=StaticSwitch(True=@n4.Direction, False=ComponentMask(None=@n5) {g=true, r=true}, Value=@n6)
   Wave Length=@n7.R
   Amplitude=@n7.R
   Wave Steepness=@n7.R
   Pos=@n8
   wavesin 1=@n16.R
   wavesin 2=@n16.R
   wavecos 1=@n13.R
   wavecos 2=@n13.R
  ) {desc=4, materialFunction=&GerstnerWave_Deep_WithZTrace}
  n18 = ComponentMask(None=Add(A=@n17.Normal, B=@n14.Normal)) {g=true, r=true}
  AppendVector(A=@n18, B=Constant() {r=1})
  n19 = MaterialFunctionCall(Wave Index=Add(A=Constant(), B=@n2), Water Body Index=@n3) {materialFunction=&GetWaveData}
  n20 = CollectionParameter() {collection=&MPC_Water, parameterName=k0}
  n21 = MaterialFunctionCall(Float4=StaticSwitch(True=@n19.Params, False=CollectionParameter() {collection=&MPC_Water, parameterName=w0}, Value=@n6)) {materialFunction=&BreakOutFloat4Components}
  n22 = MaterialFunctionCall(Wave Index=Add(A=Constant() {r=1}, B=@n2), Water Body Index=@n3) {materialFunction=&GetWaveData}
  n23 = CollectionParameter() {collection=&MPC_Water, parameterName=k1}
  n24 = MaterialFunctionCall(Float4=StaticSwitch(True=@n22.Params, False=CollectionParameter() {collection=&MPC_Water, parameterName=w1}, Value=@n6)) {materialFunction=&BreakOutFloat4Components}
  n25 = MaterialFunctionCall(Wave Index=Add(A=Constant() {r=2}, B=@n2), Water Body Index=@n3) {materialFunction=&GetWaveData}
  n26 = CollectionParameter() {collection=&MPC_Water, parameterName=k2}
  n27 = MaterialFunctionCall(Float4=StaticSwitch(True=@n25.Params, False=CollectionParameter() {collection=&MPC_Water, parameterName=w2}, Value=@n6)) {materialFunction=&BreakOutFloat4Components}
  n28 = Custom(x=@MaterialExpressionMaterialFunctionCall_0.Result) {
   code="cos(x);"
   description="cos(x)"
   outputType=CMOT_Float4
   inputs [{inputName=x, input {expression=&self:MaterialExpressionMaterialFunctionCall_0}}]
  }
  n29 = MaterialFunctionCall(Float4=@n28) {materialFunction=&BreakOutFloat4Components}
  n30 = Custom(x=@MaterialExpressionMaterialFunctionCall_31.Result) {
   code="cos(x);"
   description="cos(x)"
   outputType=CMOT_Float4
   inputs [{inputName=x, input {expression=&self:MaterialExpressionMaterialFunctionCall_31}}]
  }
  n31 = MaterialFunctionCall(Float4=@n30) {materialFunction=&BreakOutFloat4Components}
  n32 = MaterialFunctionCall(
   Z Only Offset=@n1
   2D Direction Vector=StaticSwitch(True=@n25.Direction, False=ComponentMask(None=@n26) {g=true, r=true}, Value=@n6)
   Wave Length=@n27.R
   Amplitude=@n27.R
   Wave Steepness=@n27.R
   Pos=@n8
   wavesin 1=@n41.B
   wavesin 2=@n38.B
   wavecos 1=@n29.B
   wavecos 2=@n31.B
  ) {desc=2, materialFunction=&GerstnerWave_Deep_WithZTrace}
  n33 = MaterialFunctionCall(Wave Index=Add(A=Constant() {r=3}, B=@n2), Water Body Index=@n3) {materialFunction=&GetWaveData}
  n34 = CollectionParameter() {collection=&MPC_Water, parameterName=k3}
  n35 = MaterialFunctionCall(Float4=StaticSwitch(True=@n33.Params, False=CollectionParameter() {collection=&MPC_Water, parameterName=w3}, Value=@n6)) {materialFunction=&BreakOutFloat4Components}
  n36 = MaterialFunctionCall(
   Z Only Offset=@n1
   2D Direction Vector=StaticSwitch(True=@n33.Direction, False=ComponentMask(None=@n34) {g=true, r=true}, Value=@n6)
   Wave Length=@n35.R
   Amplitude=@n35.R
   Wave Steepness=@n35.R
   Pos=@n8
   wavesin 1=@n41.A
   wavesin 2=@n38.A
   wavecos 1=@n29.A
   wavecos 2=@n31.A
  ) {desc=3, materialFunction=&GerstnerWave_Deep_WithZTrace}
  MaterialExpressionMaterialFunctionCall_31 = MaterialFunctionCall(X=@n42.Wave Pos 2, Y=@n39.Wave Pos 2, Z=@n32.Wave Pos 2, A=@n36.Wave Pos 2) {materialFunction=&MakeFloat4}
  n37 = Custom(x=@MaterialExpressionMaterialFunctionCall_31.Result) {
   code="sin(x);"
   description="Sin(x)"
   outputType=CMOT_Float4
   inputs [{inputName=x, input {expression=&self:MaterialExpressionMaterialFunctionCall_31}}]
  }
  n38 = MaterialFunctionCall(Float4=@n37) {materialFunction=&BreakOutFloat4Components}
  n39 = MaterialFunctionCall(
   Z Only Offset=@n1
   2D Direction Vector=StaticSwitch(True=@n22.Direction, False=ComponentMask(None=@n23) {g=true, r=true}, Value=@n6)
   Wave Length=@n24.R
   Amplitude=@n24.R
   Wave Steepness=@n24.R
   Pos=@n8
   wavesin 1=@n41.G
   wavesin 2=@n38.G
   wavecos 1=@n29.G
   wavecos 2=@n31.G
  ) {desc=1, materialFunction=&GerstnerWave_Deep_WithZTrace}
  MaterialExpressionMaterialFunctionCall_0 = MaterialFunctionCall(X=@n42.Wave Pos 1, Y=@n39.Wave Pos 1, Z=@n32.Wave Pos 1, A=@n36.Wave Pos 1) {materialFunction=&MakeFloat4}
  n40 = Custom(x=@MaterialExpressionMaterialFunctionCall_0.Result) {
   code="sin(x);"
   description="Sin(x)"
   outputType=CMOT_Float4
   inputs [{inputName=x, input {expression=&self:MaterialExpressionMaterialFunctionCall_0}}]
  }
  n41 = MaterialFunctionCall(Float4=@n40) {materialFunction=&BreakOutFloat4Components}
  n42 = MaterialFunctionCall(
   Z Only Offset=@n1
   2D Direction Vector=StaticSwitch(True=@n19.Direction, False=ComponentMask(None=@n20) {g=true, r=true}, Value=@n6)
   Wave Length=@n21.R
   Amplitude=@n21.R
   Wave Steepness=@n21.R
   Pos=@n8
   wavesin 1=@n41.R
   wavesin 2=@n38.R
   wavecos 1=@n29.R
   wavecos 2=@n31.R
  ) {desc=0, materialFunction=&GerstnerWave_Deep_WithZTrace}
  n43 = ComponentMask(None=Add(A=Add(A=@n42.Normal, B=@n39.Normal), B=Add(A=@n32.Normal, B=@n36.Normal))) {g=true, r=true}
  n44 = MaterialFunctionCall(
   Fade Length=ScalarParameter() {defaultValue=15000, group=Distance Blending, parameterName=Far Fade Length}
   Fade Offset=ScalarParameter() {defaultValue=7500, group=Distance Blending, parameterName=Far Fade Start}
  ) {materialFunction=&CameraDepthFade}
  n45 = OneMinus(None=@n44.Result)
  n46 = Multiply(A=@n18, B=Multiply(A=@n45, B=@n45))
  n47 = Add(A=@n43, B=@n46)
  n48 = VertexInterpolator(VS=@n47)
  n49 = Add(A=VertexInterpolator(VS=@n43).PS, B=@n46)
  n50 = FunctionInput(Preview=StaticBool()) {
   bUsePreviewValueAsDefault=true
   description="If Normals use VS Interp is checked,  this allows forcing all 6 waves to use the interpolation."
   inputName=All Normals VS
   inputType=FunctionInput_StaticBool
   sortPriority=12
  }
  n51 = FunctionInput(Preview=StaticBool()) {
   bUsePreviewValueAsDefault=true
   description="If this is true, the first 4 waves will have their normal computed on the vertex shader."
   inputName=Normals use VS Interpolation
   inputType=FunctionInput_StaticBool
   sortPriority=11
  }
  n52 = AppendVector(A=StaticSwitch(True=StaticSwitch(True=@n48.PS, False=@n49, Value=@n50), False=@n47, Value=@n51), B=Constant() {r=1})
  n53 = Constant3Vector() {constant=<rgba 0 0 1 0>}
  n54 = QualitySwitch(Default=@n52, Low=@n53, High=@n52, Medium=@n52)
  FunctionOutput(None=ShadingPathSwitch(Default=@n54, Deferred=@n54, Forward=@n53, Mobile=@n53)) {outputName=Normal}
  n55 = Add(A=@n42.WPO, B=@n39.WPO)
  n56 = Add(A=@n32.WPO, B=@n36.WPO)
  FunctionOutput(None=Add(A=Add(A=@n55, B=@n56), B=Add(A=@n17.WPO, B=@n14.WPO)) {desc=WPO}) {outputName=WPO, sortPriority=1}
  FunctionOutput(None=@n44.Result) {outputName=Distance Mask, sortPriority=2}
  Multiply()
 ]}
