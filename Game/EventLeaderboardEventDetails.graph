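# Widget Blueprint derived from FortShowdownDetailView that shows the header text for a tournament event, judging by its name on the event leaderboard.
# On Refresh Data, it breaks the tournament display info to set event and session names, though the session text starts collapsed.
# The event name is the uppercased title lines, or the series point leaderboard name when EventId is empty and that leaderboard is available.
# A third line shows uppercased region, prefixed by window round when present.
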WidgetBlueprint /Game/EventLeaderboardEventDetails
graphs [EventGraph {nodes [
   Event Refresh Data() {is=title K2Node_Event}
   n1 = IsEmpty(InString=GetEventId!(self:Fort Showdown Detail View Object Reference) -> (ReturnValue:String))
   n2 = IsSeriesPointLeaderboardAvailable!(self:Fort Showdown Detail View Object Reference) -> (ReturnValue:bool)
   n3 = Break Fort Tournament Display Info!(FortTournamentDisplayInfo:"Fort Tournament Display Info Structure (by ref)"=Get TournamentDisplayInfo() -> (TournamentDisplayInfo:Fort Tournament Display Info Structure)) -> (
    TitleLine1:Text
    TitleLine2:Text
    ScheduleInfo:Text
    FlavorDescription:Text
    DetailsDescription:Text
    ShortFormatTitle:Text
    LongFormatTitle:Text
    BackgroundTitle:Text
    PinScoreRequirement:int
    PinEarnedText:Text
    BaseColor:Linear Color Structure
    PrimaryColor:Linear Color Structure
    SecondaryColor:Linear Color Structure
    HighlightColor:Linear Color Structure
    TitleColor:Linear Color Structure
    ShadowColor:Linear Color Structure
    BackgroundLeftColor:Linear Color Structure
    BackgroundRightColor:Linear Color Structure
    BackgroundTextColor:Linear Color Structure
    PosterFadeColor:Linear Color Structure
    AlertText:Text
    SeriesPointLeaderboardName:Text
    AlertType:EFortTournamentAlertType Enum
    RoundDisplayNames:Array of Texts
   ) {is=title K2Node_BreakStruct}
   n4 = TextToUpper!(InText:"Text (by ref)"=@n3) -> (ReturnValue:Text)
   n5 = TextToUpper(InText=@n3.TitleLine2)
   > "SetText (Text)"(
    self=Get Text_EventName() -> (Text_EventName:Common Text Object Reference)
    InText=Select(
     Option 0:Text=Format Text!(Format="LOCTEXT(\"{0} {1}\")", 0:Text=@n4, 1:Text=@n5) -> (Result)
     Option 1:Text=@n3.SeriesPointLeaderboardName
     Index=AND Boolean(A=@n1, B=@n2)
    ) -> (ReturnValue:Text)
   )
   > "SetText (Text)"(self=Get Text_SessionName() -> (Text_SessionName:Common Text Object Reference), InText=@n3.ShortFormatTitle)
   n6 = GetWindowRoundDisplayString!(self:Fort Showdown Detail View Object Reference) -> (ReturnValue:Text)
   n7 = TextToUpper(InText=GetCurrentRegionName!(self:Fort Showdown Detail View Object Reference) -> (ReturnValue:Text))
   > "SetText (Text)"(
    self=Get Text_RoundAndRegionName() -> (Text_RoundAndRegionName:Common Text Object Reference)
    InText=SelectText(
     A=Format Text!(Format="LOCTEXT(\"({Region Name})\")", Region Name:Text=@n7) -> (Result)
     B=Format Text!(Format="LOCTEXT(\"{Window Round} ({Region Name})\")", Window Round:Text=@n6, Region Name=@n7) -> (Result)
     bPickA=TextIsEmpty(InText=@n6)
    )
   )
   Event PreConstruct()
   Event Construct()
   Event Tick()
  ]}]
variables []
widgets {
 info {inheritedWidgetCount=6, parentClass=/Script/FortniteUI.FortShowdownDetailView, rootWidgetClass=&Overlay, widgetCount=14}
 widgets [
  Overlay_0 {widgetClassPath=&Overlay, widget=Overlay_0}
  SizeBox_0 {
   widgetClassPath=&SizeBox
   parent=Overlay_0
   widget=SizeBox_0
   slot {class=OverlaySlot, properties {horizontalAlignment=HAlign_Fill, verticalAlignment=VAlign_Fill}}
  }
  ScaleBox_0 {
   widgetClassPath=&ScaleBox
   parent=SizeBox_0
   widget=ScaleBox_0
   slot {class=SizeBoxSlot, properties {horizontalAlignment=HAlign_Left, verticalAlignment=VAlign_Top}}
  }
  VerticalBox_0 {widgetClassPath=&VerticalBox, parent=ScaleBox_0, widget=VerticalBox_0, slot {class=ScaleBoxSlot}}
  VerticalBox_1 {
   widgetClassPath=&VerticalBox
   parent=VerticalBox_0
   widget=VerticalBox_1
   slot {class=VerticalBoxSlot, properties {padding=<ltrb 0 -10 0 -10>}}
  }
  Text_EventName {
   widgetClassPath=&CommonTextBlock
   bIsVariable=true
   parent=VerticalBox_1
   widget=Text_EventName
   properties {
    text=WINTER ROYALE
    font {
     bForceMonospaced=false
     bMaterialIsStencil=false
     fontMaterial=null
     fontObject=&Burbank
     letterSpacing=0
     monospacedWidth=1
     size=59
     skewAmount=0
     typefaceFontName=BurbankRegular-Black
     outlineSettings {
      bApplyOutlineToDropShadows=false
      bMiteredCorners=false
      bSeparateFillAlpha=false
      outlineMaterial=null
      outlineSize=0
      outlineColor=<rgba 0 0 0 1>
     }
    }
    shadowOffset=<0 0>
   }
   slot {class=VerticalBoxSlot}
  }
  Text_RoundAndRegionName {
   widgetClassPath=&CommonTextBlock
   bIsVariable=true
   parent=VerticalBox_1
   widget=Text_RoundAndRegionName
   properties {
    text="First Round (NA)"
    font {
     bForceMonospaced=false
     bMaterialIsStencil=false
     fontMaterial=null
     fontObject=&Burbank
     letterSpacing=0
     monospacedWidth=1
     size=36
     skewAmount=0
     typefaceFontName=Burbank-Black
     outlineSettings {
      bApplyOutlineToDropShadows=false
      bMiteredCorners=false
      bSeparateFillAlpha=false
      outlineMaterial=null
      outlineSize=0
      outlineColor=<rgba 0 0 0 1>
     }
    }
    shadowOffset=<0 0>
   }
   slot {class=VerticalBoxSlot, properties {padding=<ltrb 0 -12 0 -5>}}
  }
  Text_SessionName {
   widgetClassPath=&CommonTextBlock
   bIsVariable=true
   parent=VerticalBox_0
   widget=Text_SessionName
   properties {
    text=ROUND 1
    visibility=Collapsed
    font {
     bForceMonospaced=false
     bMaterialIsStencil=false
     fontMaterial=null
     fontObject=&Burbank
     letterSpacing=0
     monospacedWidth=1
     size=48
     skewAmount=0
     typefaceFontName=Burbank-Black
     outlineSettings {
      bApplyOutlineToDropShadows=false
      bMiteredCorners=false
      bSeparateFillAlpha=false
      outlineMaterial=null
      outlineSize=0
      outlineColor=<rgba 0 0 0 1>
     }
    }
    shadowOffset=<0 0>
   }
   slot {class=VerticalBoxSlot, properties {padding=<ltrb 0 -10 0 -10>}}
  }
  Text_DebugId {
   widgetClassPath=&TextBlock
   bInherited=true
   parent=Overlay_0
   widget=Text_DebugId
   properties {
    autoWrapText=true
    justification=Center
    text=EventId: Event EventWindowId: EventWindow
    wrapTextAt=210
    wrappingPolicy=AllowPerCharacterWrapping
    colorAndOpacity {colorUseRule=UseColor_Specified, specifiedColor=<rgba 0 0 0 1>}
    font {
     bForceMonospaced=false
     bMaterialIsStencil=false
     fontMaterial=null
     fontObject=&Roboto
     letterSpacing=0
     monospacedWidth=1
     size=12
     skewAmount=0
     typefaceFontName=Bold
     outlineSettings {
      bApplyOutlineToDropShadows=false
      bMiteredCorners=false
      bSeparateFillAlpha=false
      outlineMaterial=null
      outlineSize=2
      outlineColor=<rgba 1 1 1 1>
     }
    }
   }
   slot {class=OverlaySlot, properties {horizontalAlignment=HAlign_Fill, verticalAlignment=VAlign_Center}}
  }
  Text_TimeRemaining {widgetClassPath=&CommonTextBlock, bInherited=true, widget=null}
  Image_PosterFront {widgetClassPath=/Script/EpicCMSUIFramework.EpicCMSImage, bInherited=true, widget=null}
  Image_PosterBack {widgetClassPath=&EpicCMSImage, bInherited=true, widget=null}
  Image_LoadingScreen {widgetClassPath=&EpicCMSImage, bInherited=true, widget=null}
  Image_PlaylistTile {widgetClassPath=&EpicCMSImage, bInherited=true, widget=null}
 ]
}
