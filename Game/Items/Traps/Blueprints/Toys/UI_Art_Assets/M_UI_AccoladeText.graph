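# M_UI_AccoladeText is a translucent UI material, named for accolade text, that builds a two-color vertical gradient.
# The inverted V texture coordinate drives a lerp between two vector parameters, running from Color1 (muted red by default) at V=0 to Color2 (white by default) at V=1.
# The graph samples no textures and also holds a standalone constant of one.
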Material /Game/Items/Traps/Blueprints/Toys/UI_Art_Assets/M_UI_AccoladeText
graph {
 properties {blendMode=BLEND_Translucent, materialDomain=MD_UI, pixelDepthOffsetMode=PDOM_Legacy, usedShadingModels=Default Lit}
 nodes [
  Constant() {r=1}
  n1 = TextureCoordinate()
  LinearInterpolate(
   A=VectorParameter() {parameterName=Color2, defaultValue=<rgba 1 1 1 1>}.RGB
   B=VectorParameter() {parameterName=Color1, defaultValue=<rgba 0.7448 0.1629 0.1629 1>}.RGB
   Alpha=OneMinus(None=ComponentMask(None=@n1) {g=true})
  )
 ]
}
