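# Material function that outputs a pseudo-random flickering float, RandomFlickerFloat, for use in materials.
# Six DebugTimeSine calls run at different speeds, each set by a scalar parameter in the RandomFloatFlicker group.
# A sum of two waves and a difference of two others are blended by a lerp whose alpha is the product of the last pair.
# The blended result is saturated to the 0–1 range before output.
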MaterialFunction /Game/MF__FloatFlickerRandom_Generic
graph {nodes [
  n1 = Add(
   A=MaterialFunctionCall(Speed=ScalarParameter() {defaultValue=4.75, group=RandomFloatFlicker, parameterName=RandomFlicker_Value01}) {materialFunction=&DebugTimeSine}.Result
   B=MaterialFunctionCall(Speed=ScalarParameter() {defaultValue=3.56, group=RandomFloatFlicker, parameterName=RandomFlicker_Value02}) {materialFunction=&DebugTimeSine}.Result
  )
  n2 = Subtract(
   A=MaterialFunctionCall(Speed=ScalarParameter() {defaultValue=0.23, group=RandomFloatFlicker, parameterName=RandomFlicker_Value03}) {materialFunction=&DebugTimeSine}.Result
   B=MaterialFunctionCall(Speed=ScalarParameter() {defaultValue=0.75, group=RandomFloatFlicker, parameterName=RandomFlicker_Value04}) {materialFunction=&DebugTimeSine}.Result
  )
  n3 = Multiply(
   A=MaterialFunctionCall(Speed=ScalarParameter() {defaultValue=0.33, group=RandomFloatFlicker, parameterName=RandomFlicker_Value05}) {materialFunction=&DebugTimeSine}.Result
   B=MaterialFunctionCall(Speed=ScalarParameter() {defaultValue=0.647, group=RandomFloatFlicker, parameterName=RandomFlicker_Value06}) {materialFunction=&DebugTimeSine}.Result
  )
  FunctionOutput(None=Saturate(None=LinearInterpolate(A=@n1, B=@n2, Alpha=@n3))) {outputName=RandomFlickerFloat}
 ]}
