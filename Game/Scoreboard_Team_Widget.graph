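# Team block of the minigame scoreboard, showing a placement medal, team name, color bar, optional win count and a grid of player rows.
# UpdateTeamData fills these from a bucket row, reusing pooled Scoreboard_UI_Row widgets by index, creating missing ones and collapsing extras.
# Standing picks the medal switcher page, hidden for in-game score display.
# Construct applies the stored bucket, later updates overwrite it, and helpers clear rows or find one by PlayerState.
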WidgetBlueprint /Game/Scoreboard_Team_Widget
graphs [
 ClearTeamPlayers {nodes [
   ClearTeamPlayers() {is=title K2Node_FunctionEntry}
   > Clear(TargetArray:Array of Scoreboard UI Row Object References=Get TeamPlayers() -> (TeamPlayers:Array of Scoreboard UI Row Object References))
   > ClearChildren(self=Get TeamList() -> (TeamList:Grid Panel Object Reference))
  ]}
 EventGraph {nodes [
   Event Construct()
   > UpdateTeamData!(
    execute
    self
    TeamBucketData:Minigame Scoreboard Bucket Row Structure=Get TeamScoreboardBucket() -> (TeamScoreboardBucket:Minigame Scoreboard Bucket Row Structure)
    ShowVictoryScore:bool=Get bShowVictoryScore() -> (bShowVictoryScore:bool)
   ) -> (then)
   > Set UpdatedFromConstruct(UpdatedFromConstruct:bool=false) -> (Output_Get:bool)
   Event PreConstruct()
  ]}
 GetWidgetForPlayerState {nodes [
   GetWidgetForPlayerState!() -> (then, PlayerState:Player State Object Reference) {is=title K2Node_FunctionEntry}
   > n1 = For Each Loop(Array:Array of Scoreboard UI Row Object References=Get TeamPlayers()) -> (Array Element:Scoreboard UI Row Object Reference)
   Return Node!(execute=@n1.Completed, Widget:Widget Object Reference)
   > Is Valid(InputObject=@n1)
   n2 = Get PlayerBucketData!(self:Scoreboard UI Row Object Reference=@n1) -> (PlayerBucketData:Minigame Scoreboard Player Row Structure)
   > Branch(Condition="Equal (Object)"(
     A=Get PlayerState() -> (PlayerState:Player State Object Reference)
     B=Break Minigame Scoreboard Player Row!(MinigameScoreboardPlayerRow:"Minigame Scoreboard Player Row Structure (by ref)"=@n2) -> (PlayerState:Fort Player State Object Reference, bHasLeftGame:bool=~, bHideFromDisplay:bool=~) {is=title K2Node_BreakStruct}
    ) {is=title K2Node_CallFunction})
   > Return Node(Widget=@n1)
  ]}
 MakePlayerRow {nodes [
   n1 = MakePlayerRow!() -> (then, PlayerData:Minigame Scoreboard Player Row Structure) {is=title K2Node_FunctionEntry}
   > n2 = Create Widget!(
    execute:Exec
    Class:User Widget Class Reference=/Game/Items/Traps/Blueprints/Toys/Scoreboard_UI_Row.Scoreboard_UI_Row_C
    OwningPlayer:Player Controller Object Reference=GetOwningPlayer()
    bDisplayWins:bool=False
    PlayerBucketData:Minigame Scoreboard Player Row Structure=@n1
    bIngameScorerow:bool=Get bIngameTeamScore() -> (bIngameTeamScore:bool)
    bIsTeamGame:bool=true
    MedalStanding:int=0
    TeamColorParam:double=Get TeamColorParam() -> (TeamColorParam:double)
    WinCondition:EMinigameWinCondition Enum=MostRoundWins
    bShowCumulativeScoreboard:bool=False
   ) -> (then:Exec, ReturnValue:Scoreboard UI Row Object Reference)
   > n3 = Add(TargetArray:Array of Scoreboard UI Row Object References=Get TeamPlayers(), NewItem:"Scoreboard UI Row Object Reference (by ref)"=@n2) {is=title K2Node_CallArrayFunction}
   > n4 = AddChild(self=Get TeamList(), Content=@n2)
   > n5 = Cast To GridSlot!(execute, Object=@n4) -> (then, CastFailed, AsGrid Slot:Grid Slot Object Reference) {is=title K2Node_DynamicCast}
   > SetRow!(execute, self:Grid Slot Object Reference=@n5, InRow:int=@n3) -> (then)
   > SetHorizontalAlignment(self:Grid Slot Object Reference=@n5, InHorizontalAlignment=HAlign_Fill)
  ]}
 UpdateMedalStanding {nodes [
   n1 = UpdateMedalStanding!() -> (then, Standing:int) {is=title K2Node_FunctionEntry}
   > SetVisibility(
    self=Get PlacementIndicatorBox() -> (PlacementIndicatorBox:Size Box Object Reference)
    InVisibility=Select(Option 0=SelfHitTestInvisible, Option 1=Collapsed, Index=Get bIngameTeamScore())
   )
   n2 = Get PlacementIndicator_Switcher() -> (PlacementIndicator_Switcher:Common Widget Switcher Legacy Object Reference)
   n3 = GetNumWidgets!(self:Widget Switcher Object Reference=@n2) -> (ReturnValue:int=0)
   n4 = int - int(A=@n3, B=1) {is=title K2Node_CallFunction}
   > SetActiveWidgetIndex(
    self=@n2
    Index=Select(
     Option 0:int=@n3
     Option 1:int="Clamp (Integer)"(Value=@n1, Min=0, Max=@n4) -> (ReturnValue=0)
     Index="integer >= integer"!(A:int=@n1, B:int=0) -> (ReturnValue:bool) {is=title K2Node_CallFunction}
    ) -> (ReturnValue:int)
   )
  ]}
 UpdateTeamData {nodes [
   n1 = UpdateTeamData!() -> (then, TeamBucketData:Minigame Scoreboard Bucket Row Structure, ShowVictoryScore:bool=~) {is=title K2Node_FunctionEntry}
   > n2 = Sequence!(execute) -> (then_0, then_1, then_2:Exec, then_3:Exec, then_4:Exec)
   n3 = Branch(execute=@n2.then_1, Condition=@n1.ShowVictoryScore)
   n4 = Get TeamScoreBox() -> (TeamScoreBox:Size Box Object Reference)
   > SetVisibility(self=@n4, InVisibility=SelfHitTestInvisible)
   n5 = Get TeamBucketData() -> (TeamBucketData:Minigame Scoreboard Bucket Row Structure)
   > "SetText (Text)"(
    self=Get TeamScoreText() -> (TeamScoreText:Text Object Reference)
    InText="To Text (Integer)"(Value=Break Minigame Scoreboard Bucket Row!(MinigameScoreboardBucketRow:"Minigame Scoreboard Bucket Row Structure (by ref)"=@n5) -> (
      Standing:int
      BucketIndex:int
      PlayerRows:Array of Minigame Scoreboard Player Row Structures
      Name:Text
      TeamColorIndex:int
      Values:Array of Minigame Scoreboard Value Structures
      WinCount:int
      bHighlight:bool=~
     ) {is=title K2Node_BreakStruct}.WinCount)
   )
   SetVisibility(execute=@n3.else, self=@n4, InVisibility=Collapsed)
   n6 = For Each Loop(Exec=@n2.then_2, Array=Get TeamPlayers())
   n7 = Length(TargetArray:Array of Minigame Scoreboard Player Row Structures=Break Minigame Scoreboard Bucket Row(MinigameScoreboardBucketRow=Get TeamBucketData()) -> (bHighlight=~).PlayerRows)
   > SetVisibility(
    self=@n6
    InVisibility=Select(
     Option 0:ESlateVisibility Enum=Collapsed
     Option 1:ESlateVisibility Enum=SelfHitTestInvisible
     Index="integer < integer"!(A:int=@n6.Array Index, B:int=@n7) -> (ReturnValue:bool) {is=title K2Node_CallFunction}
    ) -> (ReturnValue:ESlateVisibility Enum)
   )
   n8 = Break Minigame Scoreboard Bucket Row(MinigameScoreboardBucketRow=Get TeamBucketData()) -> (bHighlight=~)
   "SetText (Text)"(execute=@n2.then_0, self=Get TeamName() -> (TeamName:Text Object Reference), InText=@n8.Name)
   > n9 = GetDynamicMaterial(self:Border Object Reference=Get TeamColor_IndicatorBar() -> (TeamColor_IndicatorBar:Border Object Reference))
   > SetScalarParameterValue(self=@n9, ParameterName=TeamColor, Value=Get TeamColorParam())
   > UpdateMedalStanding!(execute, self, Standing:int=@n8) -> (then) {is=title K2Node_CallFunction}
   n10 = For Each Loop(
    Exec=@n2.then_3
    Array:Array of Minigame Scoreboard Player Row Structures=Break Minigame Scoreboard Bucket Row(MinigameScoreboardBucketRow=Get TeamBucketData()) -> (bHighlight=~).PlayerRows
   ) -> (Array Element:Minigame Scoreboard Player Row Structure)
   > n11 = Branch(Condition=Is Valid Index!(TargetArray:Array of Scoreboard UI Row Object References=Get TeamPlayers(), IndexToTest:int=@n10.Array Index) -> (ReturnValue:bool))
   > UpdatePlayerData!(
    execute
    self:Scoreboard UI Row Object Reference="Get (a copy)"(Array:Array of Scoreboard UI Row Object References=Get TeamPlayers(), Dimension 1=@n10.Array Index) -> (Output:Scoreboard UI Row Object Reference)
    PlayerBucketData:Minigame Scoreboard Player Row Structure=@n10
   ) -> (then)
   MakePlayerRow!(execute=@n11.else, self, PlayerData:Minigame Scoreboard Player Row Structure=@n10) -> (then) {is=title K2Node_CallFunction}
   n12 = Branch(execute=@n2.then_4, Condition=Get UpdatedFromConstruct() -> (UpdatedFromConstruct:bool))
   Set TeamScoreboardBucket(execute=@n12.else, TeamScoreboardBucket:Minigame Scoreboard Bucket Row Structure=Get TeamBucketData()) -> (Output_Get:Minigame Scoreboard Bucket Row Structure)
  ]}
]
variables [
 bIngameTeamScore {type=boolean}
 bShowVictoryScore {type=boolean}
 TeamColorParam {type=number}
 TeamPlayers {type="array<Scoreboard_UI_Row_C>", default []}
 TeamScoreboardBucket {
  type=MinigameScoreboardBucketRow
  default {name="", bHighlight=false, bucketIndex=-1, standing=-1, teamColorIndex=-1, winCount=0, playerRows [], values []}
 }
 UpdatedFromConstruct {type=boolean, default=true}
]
widgets {
 info {inheritedWidgetCount=0, parentClass=&UserWidget, rootWidgetClass=&VerticalBox, widgetCount=39}
 widgets [
  VerticalBox_0 {widgetClassPath=&VerticalBox, widget=VerticalBox_0}
  HorizontalBox_0 {
   widgetClassPath=&HorizontalBox
   parent=VerticalBox_0
   widget=HorizontalBox_0
   slot {class=VerticalBoxSlot, properties {verticalAlignment=VAlign_Top, size {sizeRule=Fill, value=1}}}
  }
  PlacementIndicatorBox {
   widgetClassPath=&SizeBox
   bIsVariable=true
   parent=HorizontalBox_0
   widget=PlacementIndicatorBox
   properties {bOverride_WidthOverride=true, heightOverride=128, visibility=Collapsed, widthOverride=96}
   slot {class=HorizontalBoxSlot}
  }
  HorizontalBox_4 {widgetClassPath=&HorizontalBox, parent=PlacementIndicatorBox, widget=HorizontalBox_4, slot {class=SizeBoxSlot}}
  Overlay_2 {
   widgetClassPath=&Overlay
   parent=HorizontalBox_4
   widget=Overlay_2
   slot {class=HorizontalBoxSlot, properties {size {sizeRule=Fill, value=1}}}
  }
  PlacementIndicator_Switcher {
   widgetClassPath=&CommonWidgetSwitcherLegacy
   bIsVariable=true
   parent=Overlay_2
   widget=PlacementIndicator_Switcher
   slot {class=OverlaySlot, properties {horizontalAlignment=HAlign_Fill, verticalAlignment=VAlign_Fill}}
  }
  Scoreboard_Gold {
   widgetClassPath=/Game/Items/Traps/Blueprints/Toys/UI_Art_Assets/Scoreboard_Gold
   bIsVariable=true
   parent=PlacementIndicator_Switcher
   widget=Scoreboard_Gold
   slot {class=WidgetSwitcherSlot}
  }
  Scoreboard_Silver {
   widgetClassPath=/Game/Items/Traps/Blueprints/Toys/UI_Art_Assets/Scoreboard_Silver
   bIsVariable=true
   parent=PlacementIndicator_Switcher
   widget=Scoreboard_Silver
   slot {class=WidgetSwitcherSlot}
  }
  Scoreboard_Bronze {
   widgetClassPath=/Game/Items/Traps/Blueprints/Toys/UI_Art_Assets/Scoreboard_Bronze
   bIsVariable=true
   parent=PlacementIndicator_Switcher
   widget=Scoreboard_Bronze
   slot {class=WidgetSwitcherSlot}
  }
  Placement_None {widgetClassPath=&Overlay, parent=PlacementIndicator_Switcher, widget=Placement_None, slot {class=WidgetSwitcherSlot}}
  PlacementBacking_NOne {
   widgetClassPath=&Border
   parent=Placement_None
   widget=PlacementBacking_NOne
   properties {brushColor=<rgba 0 0.0417 0.5 1>, padding=<ltrb 0 0 0 0>}
   slot {class=OverlaySlot, properties {horizontalAlignment=HAlign_Fill, verticalAlignment=VAlign_Fill}}
  }
  Spacer_76 {
   widgetClassPath=&Spacer
   parent=PlacementBacking_NOne
   widget=Spacer_76
   properties {size=<1 32>}
   slot {class=BorderSlot, properties {padding=<ltrb 0 0 0 0>}}
  }
  SizeBox_1 {
   widgetClassPath=&SizeBox
   parent=Placement_None
   widget=SizeBox_1
   properties {bOverride_HeightOverride=true, bOverride_WidthOverride=true, heightOverride=6, widthOverride=24}
   slot {class=OverlaySlot, properties {horizontalAlignment=HAlign_Center, verticalAlignment=VAlign_Center}}
  }
  Image_65 {
   widgetClassPath=&Image
   bIsVariable=true
   parent=SizeBox_1
   widget=Image_65
   properties {colorAndOpacity=<rgba 0.0021 0.1248 0.6724 1>}
   slot {class=SizeBoxSlot}
  }
  Spacer_60 {widgetClassPath=&Spacer, parent=HorizontalBox_4, widget=Spacer_60, properties {size=<8 1>}, slot {class=HorizontalBoxSlot}}
  TeamNameBox {
   widgetClassPath=&SizeBox
   parent=HorizontalBox_0
   widget=TeamNameBox
   properties {bOverride_WidthOverride=true, heightOverride=144.2365, widthOverride=256}
   slot {class=HorizontalBoxSlot}
  }
  HorizontalBox_2 {widgetClassPath=&HorizontalBox, parent=TeamNameBox, widget=HorizontalBox_2, slot {class=SizeBoxSlot}}
  Overlay_0 {
   widgetClassPath=&Overlay
   parent=HorizontalBox_2
   widget=Overlay_0
   slot {class=HorizontalBoxSlot, properties {size {sizeRule=Fill, value=1}}}
  }
  Backing {
   widgetClassPath=&Border
   parent=Overlay_0
   widget=Backing
   properties {brushColor=<rgba 0 0.1441 1 0.75>, padding=<ltrb 0 0 0 0>}
   slot {class=OverlaySlot, properties {horizontalAlignment=HAlign_Fill, verticalAlignment=VAlign_Fill}}
  }
  HorizontalBox_1 {
   widgetClassPath=&HorizontalBox
   parent=Overlay_0
   widget=HorizontalBox_1
   slot {class=OverlaySlot, properties {horizontalAlignment=HAlign_Fill, verticalAlignment=VAlign_Fill}}
  }
  TeamInfo {
   widgetClassPath=&HorizontalBox
   bIsVariable=true
   parent=HorizontalBox_1
   widget=TeamInfo
   slot {class=HorizontalBoxSlot, properties {size {sizeRule=Fill, value=1}}}
  }
  Overlay_3 {widgetClassPath=&Overlay, parent=TeamInfo, widget=Overlay_3, slot {class=HorizontalBoxSlot}}
  Spacer_122 {widgetClassPath=&Spacer, parent=Overlay_3, widget=Spacer_122, properties {size=<12 1>}, slot {class=OverlaySlot}}
  SizeBox_0 {
   widgetClassPath=&SizeBox
   parent=Overlay_3
   widget=SizeBox_0
   properties {bOverride_WidthOverride=true, widthOverride=6}
   slot {class=OverlaySlot, properties {verticalAlignment=VAlign_Fill}}
  }
  Border_132 {
   widgetClassPath=&Border
   parent=SizeBox_0
   widget=Border_132
   properties {brushColor=<rgba 0 0.2159 1 0.5>, padding=<ltrb 0 0 0 0>}
   slot {class=SizeBoxSlot}
  }
  ScaleBoxTeamName {
   widgetClassPath=&ScaleBox
   parent=TeamInfo
   widget=ScaleBoxTeamName
   properties {stretchDirection=DownOnly}
   slot {class=HorizontalBoxSlot, properties {size {sizeRule=Fill, value=1}}}
  }
  TeamName {
   widgetClassPath=&TextBlock
   bIsVariable=true
   parent=ScaleBoxTeamName
   widget=TeamName
   properties {
    text=TEAM 00
    colorAndOpacity {colorUseRule=UseColor_Specified, specifiedColor=<rgba 0.521 0.8228 1 1>}
    font {
     bForceMonospaced=false
     bMaterialIsStencil=false
     fontMaterial=null
     fontObject=&BurbankSmall
     letterSpacing=0
     monospacedWidth=1
     size=18
     skewAmount=0
     typefaceFontName=Black
     outlineSettings {
      bApplyOutlineToDropShadows=false
      bMiteredCorners=false
      bSeparateFillAlpha=false
      outlineMaterial=null
      outlineSize=0
      outlineColor=<rgba 0 0 0 1>
     }
    }
   }
   slot {class=ScaleBoxSlot, properties {horizontalAlignment=HAlign_Left}}
  }
  TeamColorBox {
   widgetClassPath=&SizeBox
   parent=TeamInfo
   widget=TeamColorBox
   properties {bOverride_WidthOverride=true, widthOverride=6}
   slot {class=HorizontalBoxSlot, properties {horizontalAlignment=HAlign_Right}}
  }
  TeamColor_IndicatorBar {
   widgetClassPath=&Border
   bIsVariable=true
   parent=TeamColorBox
   widget=TeamColor_IndicatorBar
   properties {
    background {
     bIsDynamicallyLoaded=false
     drawAs=Image
     imageType=NoImage
     mirroring=NoMirror
     resourceName=None
     resourceObject=/Game/Items/Traps/Blueprints/Toys/UI_Art_Assets/M_Creative_TeamColor
     tiling=NoTile
     imageSize=<32 32>
     margin=<ltrb 0 0 0 0>
     outlineSettings {
      bUseBrushTransparency=false
      roundingType=HalfHeightRadius
      width=0
      color {colorUseRule=UseColor_Specified, specifiedColor=<rgba 0 0 0 0>}
      cornerRadii=<0 0 0 0>
     }
     tintColor {colorUseRule=UseColor_Specified, specifiedColor=<rgba 1 1 1 1>}
     uVRegion {bIsValid=false, max=<0 0>, min=<0 0>}
    }
    padding=<ltrb 0 0 0 0>
   }
   slot {class=SizeBoxSlot}
  }
  Spacer_177 {widgetClassPath=&Spacer, parent=HorizontalBox_2, widget=Spacer_177, properties {size=<8 1>}, slot {class=HorizontalBoxSlot}}
  TeamScoreBox {
   widgetClassPath=&SizeBox
   bIsVariable=true
   parent=HorizontalBox_0
   widget=TeamScoreBox
   properties {bOverride_WidthOverride=true, heightOverride=144.2365, widthOverride=128}
   slot {class=HorizontalBoxSlot}
  }
  HorizontalBox_3 {widgetClassPath=&HorizontalBox, parent=TeamScoreBox, widget=HorizontalBox_3, slot {class=SizeBoxSlot}}
  ScoreBox {
   widgetClassPath=&VerticalBox
   parent=HorizontalBox_3
   widget=ScoreBox
   slot {class=HorizontalBoxSlot, properties {size {sizeRule=Fill, value=1}}}
  }
  Border_0 {
   widgetClassPath=&Border
   parent=ScoreBox
   widget=Border_0
   properties {verticalAlignment=VAlign_Center, brushColor=<rgba 0 0.1329 0.6038 0.9>, padding=<ltrb 0 0 0 0>}
   slot {class=VerticalBoxSlot, properties {size {sizeRule=Fill, value=1}}}
  }
  ScaleBox_0 {
   widgetClassPath=&ScaleBox
   parent=Border_0
   widget=ScaleBox_0
   properties {stretchDirection=DownOnly}
   slot {class=BorderSlot, properties {verticalAlignment=VAlign_Center, padding=<ltrb 0 0 0 0>}}
  }
  TeamScoreText {
   widgetClassPath=&TextBlock
   bIsVariable=true
   parent=ScaleBox_0
   widget=TeamScoreText
   properties {
    justification=Center
    text=00
    font {
     bForceMonospaced=false
     bMaterialIsStencil=false
     fontMaterial=null
     fontObject=&Burbank
     letterSpacing=0
     monospacedWidth=1
     size=18
     skewAmount=0
     typefaceFontName=Burbank-Black
     outlineSettings {
      bApplyOutlineToDropShadows=false
      bMiteredCorners=false
      bSeparateFillAlpha=false
      outlineMaterial=null
      outlineSize=0
      outlineColor=<rgba 0 0 0 1>
     }
    }
    renderTransform {angle=0, scale=<1 1>, shear=<-8 0>, translation=<0 0>}
   }
   slot {class=ScaleBoxSlot, properties {verticalAlignment=VAlign_Top}}
  }
  Spacer_245 {widgetClassPath=&Spacer, parent=HorizontalBox_3, widget=Spacer_245, properties {size=<8 1>}, slot {class=HorizontalBoxSlot}}
  TeamList {
   widgetClassPath=&GridPanel
   bIsVariable=true
   parent=HorizontalBox_0
   widget=TeamList
   properties {columnFill [1]}
   slot {class=HorizontalBoxSlot, properties {verticalAlignment=VAlign_Top, size {sizeRule=Fill, value=1}}}
  }
  Spacer_29 {
   widgetClassPath=&Spacer
   parent=VerticalBox_0
   widget=Spacer_29
   properties {size=<8 16>}
   slot {class=VerticalBoxSlot, properties {verticalAlignment=VAlign_Top}}
  }
 ]
}
